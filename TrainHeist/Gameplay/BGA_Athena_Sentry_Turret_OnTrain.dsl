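; Train-mounted sentry turret for Train Heist that targets boarded pawns, switching to the vault's thermite explosive until detonation locks in or a timer stops it.
; Train platform, interaction and fuse events set targets, with optional periodic distance-and-state scoring.
; Each Fire call steps a team-size-scaled cadence array through overheat, warning and shots, driving lights, beam and cues.
; Pickaxe damage drops to 1; damage can trigger express train movement.
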(event Custom|DeleteComponents
  (Components|DestroyComponent (Variables|BuildingGameplayActorSentry|GetTargetActivationVolume))
  (Utilities|FlowControl|DelayUntilNextTick)
  (for _array_element (Actor|GetComponentsByClass "/Script/Engine.SceneComponent")
    (Development|PrintString (Utilities|GetDisplayName _array_element) true true "(R=1.000000,G=0.000000,B=0.025078,A=1.000000)" 1e+01)))

(event EventEndPlay
  (params (EndPlayReason :type EEndPlayReason%20Enum :default "Destroyed"))
  (Networking|SwitchHasAuthority
    (:Remote
      (Projectile|RemoveGameplayCue%28Looping%29Local self (Variables|Default|GetTurretOverheatGC) (Ability|GameplayCue|MakeGameplayCueParameters 0.0)))))

(event Custom|BindToInteractMulticast
  (params (Target :type Prop%20Vault%20Explosives%20Heistable%20Train%20Object%20Reference))
  (Default|BindEventtoOnStartedInteracting Target (delegate Custom|PlayerStartedTryingToPlantBomb)))

(event Custom|PlayerStartedTryingToPlantBomb
  (params (InteractingPawn :type Fort%20Pawn%20Object%20Reference))
  (CallFunction|CheckforLockonGameplayCue InteractingPawn)
  (bind _output_get (Variables|Default|SetCurrentTarget InteractingPawn))
  (Networking|SwitchHasAuthority
    (:Authority
      (CallFunction|SetActortoTrack _output_get)
      (CallFunction|StartFiringMulticast))))

(event Custom|BindToFuseStarted
  (params (Target :type Prop%20Vault%20Explosives%20Heistable%20Train%20Object%20Reference))
  (Default|BindEventtoOnFuseStarted Target (delegate Custom|OnFuseStarted)))

(event Custom|OnFuseStarted
  (CallFunction|CheckforLockonGameplayCue (Variables|Default|GetPropVaultExplosivesHeistableTrain))
  (CallFunction|SetActortoTrack (Variables|Default|GetPropVaultExplosivesHeistableTrain))
  (CallFunction|StartFiringMulticast)
  (Variables|Default|SetbAttackingThermite true)
  (bind _returnvalue (Utilities|Time|SetTimerbyEvent (delegate Custom|StopThermiteDetonation) (ScalableFloat|GetValueatLevel (Variables|Default|GetSF_StopTermiteDetonationTime))))
  (Variables|Default|SetStopThermiteDetonationTimerHandle _returnvalue))

(event Custom|StopThermiteDetonation
  (bind _currenttarget (Variables|Default|GetCurrentTarget))
  (Variables|Default|SetbAttackingThermite false)
  (CallFunction|ForceLostTarget)
  (Class|PropVaultExplosivesHeistableTrain|StopDetonation (Variables|Default|GetPropVaultExplosivesHeistableTrain))
  (Utilities|IsValid _currenttarget
    (:Is%20Valid
      (CallFunction|SetActortoTrack _currenttarget)
      (CallFunction|StartFiringMulticast))))

(event Custom|BindToDetonationLockedIn
  (params (Target :type Prop%20Vault%20Explosives%20Heistable%20Train%20Object%20Reference))
  (Default|BindEventtoOnDetonationLockedIn Target (delegate Custom|OnDetonationLockedIn)))

(event Custom|ShowOverheated_Multicast
  (Variables|Default|SetActive false)
  (Networking|SwitchHasAuthority
    (:Authority
      (Variables|Default|SetToggleOverheatFX true))
    (:Remote
      _))
  (Audio|SpawnSoundAttached :Sound (Variables|Default|GetTurretOverheatSound) :AttachToComponent (Variables|BuildingGameplayActorSentry|GetCameraStaticMeshComp) :AttachPointName "None" :LocationType "KeepRelativeOffset" :bStopWhenAttachedToDestroyed true :VolumeMultiplier 1.0 :PitchMultiplier 1.0 :StartTime 0.0 :bAutoDestroy true)
  (Rendering|SetVisibility (Variables|BuildingGameplayActorSentry|GetAngleRaysStaticMeshComp) false false)
  (Projectile|RemoveGameplayCue%28Looping%29Local self (Variables|Default|GetGC_TargetingBeamEndpoint) (Utilities|BeamEndpointParameters))
  (Development|PrintString "TURN OFF")
  (bind (_direction _newtrack_0) (Timeline|ColorTimeline Reverse)
    (:Update
      (CallFunction|SetLightsColor (Math|Color|Lerp(LinearColor) "(R=0.000000,G=1.000000,B=1.000000,A=1.000000)" "(R=1.000000,G=0.000000,B=0.015204,A=1.000000)" _newtrack_0))
      (Variables|Default|SetRotationDownwardOffset (Math|Float|Lerp 4e+01 0.0 _newtrack_0)))))

(event Custom|OnDetonationLockedIn
  (bind _currenttarget (Variables|Default|GetCurrentTarget))
  (Variables|Default|SetbAttackingThermite false)
  (Utilities|Time|ClearandInvalidateTimerbyHandle (Variables|Default|GetStopThermiteDetonationTimerHandle))
  (CallFunction|ForceLostTarget)
  (Utilities|IsValid _currenttarget
    (:Is%20Valid
      (CallFunction|SetActortoTrack _currenttarget)
      (CallFunction|StartFiringMulticast))))

(event EventFire
  (Utilities|IsValid (Variables|Default|GetCurrentTarget)
    (:Is%20Valid
      (bind _result (Math|Integer|IncrementInt (Variables|Default|GetFireNum))
        (:%20%20
          (Development|PrintString (Utilities|String|ToString(Integer) _result))
          (switch int (Utilities|Array|Get(acopy) (Variables|Default|GetFiringCadence_0) (Math|Integer|%25%28Integer%29 _result (Utilities|Array|Length (Variables|Default|GetFiringCadence_0))))
            (:-1
              (CallFunction|ShowOverheatedMulticast))
            (:1
              (Networking|SwitchHasAuthority
                (:Authority
                  (CallFunction|ShowAbouttoFireMulticast))))
            (:2
              (|CallParentFunction))))))
    (:Is%20Not%20Valid
      (Variables|Default|SetFireNum (- (Utilities|Array|FindItem (Variables|Default|GetFiringCadence_0) 1) (ScalableFloat|GetValueasInteger (Variables|Default|GetShotsBeforeStartFiringOnBoarded))))
      (if (Variables|Default|GetActive)
        (Variables|Default|SetActive false)
        (Networking|SwitchHasAuthority
          (:Authority
            (Variables|Default|SetToggleOverheatFX true))
          (:Remote
            _))
        (Audio|SpawnSoundAttached :Sound (Variables|Default|GetTurretOverheatSound) :AttachToComponent (Variables|BuildingGameplayActorSentry|GetCameraStaticMeshComp) :AttachPointName "None" :LocationType "KeepRelativeOffset" :bStopWhenAttachedToDestroyed true :VolumeMultiplier 1.0 :PitchMultiplier 1.0 :StartTime 0.0 :bAutoDestroy true)
        (Rendering|SetVisibility (Variables|BuildingGameplayActorSentry|GetAngleRaysStaticMeshComp) false false)
        (Projectile|RemoveGameplayCue%28Looping%29Local self (Variables|Default|GetGC_TargetingBeamEndpoint) (Utilities|BeamEndpointParameters))
        (Development|PrintString "TURN OFF")
        (Timeline|ColorTimeline Reverse)))))

(event Custom|ShowAboutToFire
  (Networking|SwitchHasAuthority
    (:Authority
      (CallFunction|ShowAbouttoFireMulticast))))

(event Custom|ShowAboutToFire_Multicast
  (Rendering|SetVisibility (Variables|BuildingGameplayActorSentry|GetAngleRaysStaticMeshComp) true false)
  (Projectile|AddGameplayCue%28Looping%29Local self (Variables|Default|GetGC_TargetingBeamEndpoint) (Utilities|BeamEndpointParameters))
  (Variables|Default|SetActive true)
  (Networking|SwitchHasAuthority
    (:Authority
      (Variables|Default|SetToggleOverheatFX false))
    (:Remote
      _))
  (Development|PrintString "TURN ON")
  (Timeline|ColorTimeline Play))

(event EventOnTeamSizeInitialized
  (|CallParentFunction)
  (CallFunction|InitFiringCadence)
  (if (ScalableFloat|GetValueasBool (Variables|Default|GetShouldModifyHealth))
    (Health|SetMaxHealthDontCreateASC (ScalableFloat|GetValueatLevel (Variables|Default|GetHealth) (Math|Conversions|ToFloat(Integer) (Variables|Default|GetCurrentMaxTeamSize))))))

(event Custom|Init
  (params (CarTurretIsAttachedTo :type Base%20Car%20Object%20Reference))
  (Utilities|FlowControl|Delay 3.0) ; Make sure all the cars have spawned
  (bind _output_get (Variables|Default|SetTrainManager (Class|TrainCar|GetTrainManager CarTurretIsAttachedTo)))
  (EventDispatchers|BindEventtoOnDamaged (delegate Custom|OnDamaged_Event))
  (for _array_element (Variables|/Script/TrainCoreRuntime.FortTrainManager|GetCars _output_get)
    (bind _asbase_car (Utilities|Casting|CastToBaseCar (Utilities|Struct|BreakTrainCarRuntimeData _array_element)))
    (Utilities|FlowControl|Sequence
      (:then_0
        (bind _asvault_car_heistable_train (Utilities|Casting|CastToVaultCar_HeistableTrain _asbase_car))
        (bind _asprop_vault_explosives_heistable_train (Utilities|Casting|CastToProp_VaultExplosives_HeistableTrain (Variables|/Script/Engine.ChildActorComponent|GetChildActor (Variables|/TrainHeist/Gameplay/VaultCar_HeistableTrain.VaultCar_HeistableTrain_C|GetProp_VaultExplosives_HeistableTrain _asvault_car_heistable_train))))
        (bind _output_get_2 (Variables|Default|SetPropVaultExplosivesHeistableTrain _asprop_vault_explosives_heistable_train))
        (CallFunction|BindtoInteractMulticast _output_get_2)
        (CallFunction|BindtoFuseStarted _output_get_2)
        (CallFunction|BindtoDetonationLockedIn _output_get_2))
      (:then_1
        (EventDispatchers|BindEventtoOnPawnBasedonMovingPlatform (Variables|/Train/Cars/BaseCar.BaseCar_C|GetFortTrainMovingPlatform _asbase_car) (delegate Custom|PlayerBoarded))
        (EventDispatchers|BindEventtoOnPawnDebasedMovingPlatform (Variables|/Train/Cars/BaseCar.BaseCar_C|GetFortTrainMovingPlatform _asbase_car) (delegate Custom|PlayerLeft))))))

(event Custom|PlayerLeft
  (params (MovingPlatform :type Fort%20Base%20Moving%20Platform%20Component%20Object%20Reference) (Pawn :type Fort%20Pawn%20Object%20Reference) (NewBase :type Actor%20Object%20Reference))
  (if (== Pawn (Variables|Default|GetCurrentTarget))
    (Utilities|Casting|CastToBaseCar NewBase
      (:CastFailed
        (Utilities|Casting|CastToBGA_DestroyedTurretWithThermite_OnTrain NewBase
          (:CastFailed
            (Utilities|Casting|CastToBGA_Athena_Sentry_Parent NewBase
              (:CastFailed
                (if (ScalableFloat|GetValueasBool (Variables|Default|GetShouldTryToFindBestTarget))
                  (bind _aspawn (Utilities|Casting|CastToPawn Pawn))
                  (Utilities|Array|RemoveItem (Variables|Default|GetTargetsOnTrain) _aspawn)
                  (CallFunction|UpdateTarget)
                  (else
                    (CallFunction|LostTargetMulticast)
                    (Variables|Default|SetCurrentTarget 0)
                    (Development|PrintString "LOST TARGET!!" true true "(R=1.000000,G=0.000000,B=0.213213,A=1.000000)" 1e+01)))))))))))

(event Collision|EventActorBeginOverlap
  (params (OtherActor :type Actor%20Object%20Reference)))

(event Custom|OnDamaged_Event
  (params (DamagedActor :type Actor%20Object%20Reference) (Damage :type Float%20%28single-precision%29) (InstigatedBy :type Controller%20Object%20Reference) (DamageCauser :type Actor%20Object%20Reference) (HitLocation :type Vector) (FHitComponent :type Primitive%20Component%20Object%20Reference) (BoneName :type Name) (Momentum :type Vector))
  (bind _trainmanager (Variables|Default|GetTrainManager))
  (Utilities|FlowControl|Sequence
    (:then_1
      (Utilities|IsValid _trainmanager
        (:Is%20Valid
          (if (Class|FortTrainManager|TrainIsMovingVariableSpeed _trainmanager)
            (Class|FortTrainManager|BeginExpressMovement _trainmanager true)))))))

(event Custom|PlayerBoarded
  (params (MovingPlatform :type Fort%20Base%20Moving%20Platform%20Component%20Object%20Reference) (Pawn :type Fort%20Pawn%20Object%20Reference))
  (if (ScalableFloat|GetValueasBool (Variables|Default|GetShouldTryToFindBestTarget))
    (Utilities|Array|AddUnique (Variables|Default|GetTargetsOnTrain) Pawn)
    (CallFunction|UpdateTarget)
    (else
      (CallFunction|SetTarget Pawn))))

(event Custom|LostTarget_Multicast
  (CallFunction|ForceLostTarget))

(event Custom|SetTarget
  (params (New%20Target :type Pawn%20Object%20Reference))
  (if (Variables|Default|GetCHEAT_IgnorePlayers)
    _
    (elif (not (Variables|Default|GetbAttackingThermite))
      (CallFunction|CheckforLockonGameplayCue New%20Target)
      (bind _output_get (Variables|Default|SetCurrentTarget New%20Target))
      (CallFunction|SetActortoTrack _output_get))))

(event Custom|StartFiring_Multicast
  (Variables|Default|SetFireNum (Utilities|Array|FindItem (Variables|Default|GetFiringCadence_0) 1))
  (CallFunction|ShowAbouttoFire))

(event EventBeginPlay
  (|CallParentFunction)
  (Utilities|Time|SetTimerbyEvent (delegate Custom|Update%20Target) 2.0 true))

(event EventTick
  (params (DeltaSeconds :type Float%20%28single-precision%29))
  (|CallParentFunction DeltaSeconds) ; This node is disabled and will not be called.\nDrag off pins to build functionality.
)

(event Custom|Update%20Target
  (bind _currenttarget (Variables|Default|GetCurrentTarget))
  (Utilities|FlowControl|Sequence
    (:then_0
      (if (ScalableFloat|GetValueasBool (Variables|Default|GetShouldTryToFindBestTarget))
        (bind _best_target (CallFunction|ChooseBestTarget (Variables|Default|GetTargetsOnTrain)))
        (bind _asfort_pawn (Utilities|Casting|CastToFortPawn _best_target)
          (:then
            (Utilities|IsValid _asfort_pawn
              (:Is%20Valid
                (if (not (FortPawn|IsDead _asfort_pawn))
                  (CallFunction|SetTarget _asfort_pawn)))
              (:Is%20Not%20Valid
                _))
            (CallFunction|LostTargetMulticast)
            (Variables|Default|SetCurrentTarget 0))
          (:CastFailed
            _))))
    (:then_1
      (Utilities|IsValid _currenttarget
        (:Is%20Valid
          (if (> (Transformation|GetHorizontalDistanceTo _currenttarget self) (ScalableFloat|GetValueatLevel (Variables|Default|GetForceLoseTargetRangeThreshold)))
            (CallFunction|LostTargetMulticast)
            (Variables|Default|SetCurrentTarget 0)
            (Development|PrintString "LOST TARGET!!" true true "(R=1.000000,G=0.000000,B=0.213213,A=1.000000)" 1e+01)))))))

(event OnComponentBeginOverlap(TargetActivationVolume)
  (params (OverlappedComponent :type Primitive%20Component%20Object%20Reference) (OtherActor :type Actor%20Object%20Reference) (OtherComp :type Primitive%20Component%20Object%20Reference) (OtherBodyIndex :type Integer) (bFromSweep :type Boolean) (SweepResult :type Hit%20Result%20Structure%20%28by%20ref%29))
  (Utilities|Casting|CastToPawn OtherActor))

(event OnComponentEndOverlap(TargetActivationVolume)
  (params (OverlappedComponent :type Primitive%20Component%20Object%20Reference) (OtherActor :type Actor%20Object%20Reference) (OtherComp :type Primitive%20Component%20Object%20Reference) (OtherBodyIndex :type Integer))
  (bind _aspawn (Utilities|Casting|CastToPawn OtherActor))
  (if (== _aspawn (Variables|BuildingGameplayActorSentry|GetActorBeingTracked))
    _))

(event Custom|ShowPickaxeFailureMessage
  (params (Target :type Controller%20Object%20Reference))
  (bind _asfort_player_pawn (Utilities|Casting|CastToFortPlayerPawn (Pawn|GetControlledPawn Target)))
  (FailedAction|UnabletoPerformAction _asfort_player_pawn true (GameplayTags|MakeLiteralGameplayTagContainer "(GameplayTags=((TagName=\"Activate.Failed.TrainHeist.UsedPickaxeOnTurret\")))")))

(fn BlueprintModifyIncomingDamage
  (params (Damage :type Float%20%28single-precision%29) (DamageTags :type Gameplay%20Tag%20Container%20Structure%20%28by%20ref%29) (EffectContext :type Gameplay%20Effect%20Context%20Handle%20Structure%20%28by%20ref%29) (EventInstigator :type Controller%20Object%20Reference) (DamageCauser :type Actor%20Object%20Reference))
  (Utilities|Casting|CastToFortWeaponPickaxeAthena DamageCauser
    (:then
      (CallFunction|ShowPickaxeFailureMessage EventInstigator)
      (return 1.0))
    (:CastFailed
      (return Damage))))

(fn CalcAdjustedFireTargetLocation
  (params (TrackedActor :type Actor%20Object%20Reference))
  (bind _newfiretargetloc (|CallParentFunction TrackedActor))
  (return (Math|Vector|SelectVector _newfiretargetloc (- _newfiretargetloc "0.000000,0.000000,3000.000000") (Variables|Default|GetActive))))

(fn CheatTrainTurrets_SetIgnorePlayers
  (params (ShouldIgnore :type Integer))
  (Variables|Default|SetCHEAT_IgnorePlayers (Math|Conversions|ToBoolean(Integer) ShouldIgnore)))

(fn CheckForLockOnGameplayCue
  (params (NewTarget :type Actor%20Object%20Reference))
  (if (!= NewTarget (Variables|Default|GetCurrentTarget))
    (bind _newtarget (Variables|GetNewTarget)
      (:then
        (GameplayCue|ExecuteGameplayCueOnActor(Burst) _newtarget (Variables|Gameplaycue|GetLockOnGameplayCue) (Ability|GameplayCue|MakeGameplayCueParameters :NormalizedMagnitude 0.0 :RawMagnitude 0.0 :Location "0,0,0" :Normal "0,0,0" :Instigator self :EffectCauser self :SourceObject _newtarget :GameplayEffectLevel 1 :AbilityLevel 1 :bReplicateLocationWhenUsingMinimalRepProxy false)))
      (:else))))

(fn ChooseBestTarget
  (params (Targets :type Array%20of%20Pawn%20Object%20References))
  (bind _self self)
  (if (> (Utilities|Array|Length Targets) 0)
    (for _array_element Targets
      (bind _score (CallFunction|GetValidityScore _array_element))
      (if (< (+ _score (Transformation|GetDistanceTo _array_element _self)) (LocalVariables|GetBestScore))
        (LocalVariables|SetBestScore (+ _score (Transformation|GetDistanceTo _array_element _self)))
        (LocalVariables|SetTheBestTarget _array_element)))
    (return (LocalVariables|GetTheBestTarget))
    (else
      (return 0))))

(fn GetTargetRotationForVisualUpdate
  (params (DeltaSeconds :type Float%20%28single-precision%29))
  (if (ScalableFloat|GetValueasBool (Variables|Default|GetRotateBarrelDownWhenNotShooting))
    (return (Math|Rotator|CombineRotators (Variables|BuildingGameplayActorSentry|GetTargetRotation) (Math|Vector|RotatorfromAxisandAngle (Math|Vector|GetRightVector (Variables|BuildingGameplayActorSentry|GetTargetRotation)) (Variables|Default|GetRotationDownwardOffset))))
    (else
      (return (Variables|BuildingGameplayActorSentry|GetTargetRotation)))))

(fn GetValidityScore
  (params (Target :type Object%20Reference))
  (Utilities|IsValid Target
    (:Is%20Valid
      (bind _asplayer_pawn_athena (Utilities|Casting|CastToPlayerPawn_Athena Target)
        (:then
          (if (Health|IsDBNO _asplayer_pawn_athena)
            _
            (elif (FortPawn|IsDead _asplayer_pawn_athena)
              (return 5e+05)
              (else
                (return 0.0)))))
        (:CastFailed
          _))
      (return 1e+05))
    (:Is%20Not%20Valid
      _)))

(fn HandleCustomImpulseFromDamage
  (params (ImpulseValue :type Vector :default "0, 0, 0") (Damage :type Float%20%28single-precision%29) (DamageTags :type Gameplay%20Tag%20Container%20Structure%20%28by%20ref%29) (HitResult :type Hit%20Result%20Structure%20%28by%20ref%29) (InstigatingController :type Controller%20Object%20Reference) (DamageCauser :type Actor%20Object%20Reference) (EffectContext :type Gameplay%20Effect%20Context%20Handle%20Structure%20%28by%20ref%29))
  (return false))

(fn InitFiringCadence
  (Utilities|Array|Clear (Variables|Default|GetFiringCadence_0))
  (Utilities|Array|Add (Variables|Default|GetFiringCadence_0) -1)
  (for _ (range 1 (+ (ScalableFloat|GetValueasInteger (Variables|Default|GetRestingShotsNum) (Math|Conversions|ToFloat(Integer) (Variables|Default|GetCurrentMaxTeamSize))) 1))
    (Utilities|Array|Add (Variables|Default|GetFiringCadence_0)))
  (bind _returnvalue (Utilities|Array|Add (Variables|Default|GetFiringCadence_0) 1))
  (if (ScalableFloat|GetValueasBool (Variables|Default|GetShouldInitFireNum))
    (Variables|Default|SetFireNum (- _returnvalue (ScalableFloat|GetValueasInteger (Variables|Default|GetShotsBeforeStartFiringOnBoarded))))
    (else
      _))
  (for _ (range 1 (+ (ScalableFloat|GetValueasInteger (Variables|Default|GetNumEmptyShotsToStartFiring) (Math|Conversions|ToFloat(Integer) (Variables|Default|GetCurrentMaxTeamSize))) 1))
    (Utilities|Array|Add (Variables|Default|GetFiringCadence_0)))
  (for _ (range 1 (+ (ScalableFloat|GetValueasInteger (Variables|Default|GetFiringShotsNum) (Math|Conversions|ToFloat(Integer) (Variables|Default|GetCurrentMaxTeamSize))) 1))
    (Utilities|Array|Add (Variables|Default|GetFiringCadence_0) 2))
  (Development|PrintString (Utilities|String|ToString(Integer) (Utilities|Array|Get(acopy) (Variables|Default|GetFiringCadence_0)))))

(fn OnRep_ToggleOverheatFX
  (bind _self self)
  (bind _turretoverheatgc (Variables|Default|GetTurretOverheatGC))
  (Networking|SwitchHasAuthority
    (:Remote
      (if (Variables|Default|GetToggleOverheatFX)
        (Projectile|AddGameplayCue%28Looping%29Local _self _turretoverheatgc (Ability|GameplayCue|MakeGameplayCueParameters :NormalizedMagnitude 0.0 :RawMagnitude 0.0 :Location "0,0,0" :Normal "0,0,0" :SourceObject (Variables|BuildingGameplayActorSentry|GetCameraStaticMeshComp) :GameplayEffectLevel 1 :AbilityLevel 1 :TargetAttachComponent (Variables|BuildingGameplayActorSentry|GetCameraStaticMeshComp) :bReplicateLocationWhenUsingMinimalRepProxy false))
        (else
          (Projectile|RemoveGameplayCue%28Looping%29Local _self _turretoverheatgc (Ability|GameplayCue|MakeGameplayCueParameters 0.0)))))))

(fn ReduceTrainComponents
  (CallFunction|DeleteComponents))

(fn SetLightsColor
  (params (InLinearColor :type Linear%20Color%20Structure))
  (bind _sm_pg_turret_base (Variables|Default|GetSM_PG_Turret_Base))
  (Rendering|Material|SetVectorParameterValueonMaterials _sm_pg_turret_base "EmissiveColor" (Math|Conversions|ToVector(LinearColor) InLinearColor))
  (Rendering|Material|SetScalarParameterValueonMaterials _sm_pg_turret_base "EmissiveScale" 1.0))

(fn ConstructionScript
  (|CallParentFunction)
  (Components|DestroyComponent (Variables|BuildingGameplayActorSentry|GetTargetActivationVolume)))
