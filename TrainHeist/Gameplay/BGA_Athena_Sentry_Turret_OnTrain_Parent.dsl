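; Parent Blueprint for the Train Heist on-train sentry turret, covering its state visuals and firing helpers.
; Emissive color follows sentry state and team affiliation, written through material parameters or custom primitive data (for Creative).
; Timelines drive open/close animation and morph scanning rays into targeting lasers via primitive data; Spy LTM hides rays from defenders.
; Functions supply capsule-adjusted aim points, aggro and wind-up delays, reactivation state and continue-shooting checks.
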(event Custom|SetStartingColor
  (bind _bisfriendly (CallFunction|IsFriendlytoLocalPlayer))
  (bind _seekingcolor (Variables|VFX|GetSeekingColor))
  (bind _seekingcolorfriendly (Variables|VFX|GetSeekingColorFriendly))
  (bind _returnvalue (Math|Color|SelectColor _seekingcolorfriendly _seekingcolor _bisfriendly))
  (bind _currentsentrystate (Variables|BuildingGameplayActorSentry|GetCurrentSentryState))
  (bind _trackingcolor (Variables|VFX|GetTrackingColor))
  (bind _returnvalue_2 (Utilities|Select :Index _currentsentrystate :PassiveIdle _returnvalue :ActiveIdle _returnvalue :Tracking _trackingcolor :Aggro _trackingcolor :ReturningToIdle _returnvalue :LocatingDamager _trackingcolor))
  (bind _sm_pg_turret_base (Variables|Default|GetSM_PG_Turret_Base))
  ;[0.4479 0.002217 0.3207] SetStartingColor()
  ; ChangeLightColor()
  (if (not (Variables|HandleMaterialParameters|GetbUsePrimitiveDataForMaterials))
    (Rendering|Material|SetVectorParameterValueonMaterials _sm_pg_turret_base "EmissiveColor" (Math|Conversions|ToVector(LinearColor) _returnvalue_2))
    (Rendering|Material|SetScalarParameterValueonMaterials _sm_pg_turret_base "EmissiveScale" 1.0)
    (else
      (Rendering|Material|SetCustomPrimitiveDataVector4 (Variables|Default|GetSM_PG_Turret_Base) 0 (Utilities|LinearColortoVector4 _returnvalue_2)) ; EmissiveColor = Index 0,1,2,3 (for Creative)
      (Rendering|Material|SetCustomPrimitiveDataFloat (Variables|BuildingGameplayActorSentry|GetCameraStaticMeshComp) 4 1.0) ; EmissiveScale = Index 4 (for Creative)
    )))

(event Custom|ChangeLightColor
  (bind _bisfriendly (CallFunction|IsFriendlytoLocalPlayer))
  (bind _seekingcolor (Variables|VFX|GetSeekingColor))
  (bind _seekingcolorfriendly (Variables|VFX|GetSeekingColorFriendly))
  (bind _returnvalue (Math|Color|SelectColor _seekingcolorfriendly _seekingcolor _bisfriendly))
  (bind _currentsentrystate (Variables|BuildingGameplayActorSentry|GetCurrentSentryState))
  (bind _trackingcolor (Variables|VFX|GetTrackingColor))
  (bind _returnvalue_2 (Utilities|Select :Index _currentsentrystate :PassiveIdle _returnvalue :ActiveIdle _returnvalue :Tracking _trackingcolor :Aggro _trackingcolor :ReturningToIdle _returnvalue :LocatingDamager _trackingcolor))
  (bind _sm_pg_turret_base (Variables|Default|GetSM_PG_Turret_Base))
  ;[0.4479 0.002217 0.3207] SetStartingColor()
  ; ChangeLightColor()
  (Networking|CanExecuteCosmeticEvents
    (:True
      (if (not (Variables|HandleMaterialParameters|GetbUsePrimitiveDataForMaterials))
        (Rendering|Material|SetVectorParameterValueonMaterials _sm_pg_turret_base "EmissiveColor" (Math|Conversions|ToVector(LinearColor) _returnvalue_2))
        (Rendering|Material|SetScalarParameterValueonMaterials _sm_pg_turret_base "EmissiveScale" 1.0)
        (else
          (Rendering|Material|SetCustomPrimitiveDataVector4 (Variables|Default|GetSM_PG_Turret_Base) 0 (Utilities|LinearColortoVector4 _returnvalue_2)) ; EmissiveColor = Index 0,1,2,3 (for Creative)
          (Rendering|Material|SetCustomPrimitiveDataFloat (Variables|BuildingGameplayActorSentry|GetCameraStaticMeshComp) 4 1.0) ; EmissiveScale = Index 4 (for Creative)
        )))))

(event Custom|TargetedRayScale
  (params (Direction :type ETimelineDirection%20Enum :default "Forward"))
  (bind _angleraysstaticmeshcomp (Variables|BuildingGameplayActorSentry|GetAngleRaysStaticMeshComp))
  ;[0.4321 0.2131 0.5] Targeted Ray Scale - use timelines to fold in the ray and set the scale so that it extends to the player. Undo that when false.
  (if (not (Networking|IsServer))
    (switch Utilities|FlowControl|Switch|SwitchonETimelineDirection Direction
      (:Forward
        (Utilities|DoOnceThenDo false
          (:DoOnce
            (bind (_direction _alpha _emissivescale) (Timeline|TLTargetRayScale PlayFromStart)
              (:Update
                ;;[0 0.8877 1] TlTargetRayScale Update
                (if (not (Variables|HandleMaterialParameters|GetbUsePrimitiveDataForMaterials))
                  (Rendering|Material|SetScalarParameterValueonMaterials (Variables|Default|GetSM_PG_Turret_Base) "EmissiveScale" (Utilities|Select :Index _direction :Forward _emissivescale :Backward 1.0))
                  (else
                    (Rendering|Material|SetCustomPrimitiveDataFloat (Variables|BuildingGameplayActorSentry|GetCameraStaticMeshComp) 4 (Utilities|Select :Index _direction :Forward _emissivescale :Backward 1.0)) ; EmissiveScale = Index 4 (for Creative)
                  ))
                (Rendering|Material|SetCustomPrimitiveDataFloat _angleraysstaticmeshcomp 0 (Math|Float|Lerp (Variables|SentryParentVars|GetTargetingConeAngle) 0.0 _alpha))
                (Transformation|SetRelativeScale3D _angleraysstaticmeshcomp (Math|Vector|Lerp(Vector) (Variables|SentryParentVars|GetStartingScaleRays) (Variables|BuildingGameplayActorSentry|GetAngleRaysTargetScale) _alpha))
                (Rendering|Material|SetCustomPrimitiveDataFloat _angleraysstaticmeshcomp 1 (Math|Float|Lerp (Variables|VFX|GetScanningLaserEdgeWidthEnd) (Variables|VFX|GetTargetingLaserEdgeWidthEnd) _alpha))
                (Rendering|Material|SetCustomPrimitiveDataFloat _angleraysstaticmeshcomp 2 (Math|Float|Lerp (Variables|VFX|GetScanningLaserEdgeWidthFront) (Variables|VFX|GetTargetingLaserEdgeWidthFront) _alpha))
                (Rendering|Material|SetCustomPrimitiveDataFloat _angleraysstaticmeshcomp 3 (Math|Float|Lerp (Variables|VFX|GetScanningLaserFalloff) (Variables|VFX|GetTargetingLaserFalloff) _alpha))
                (bind (_returnvalue_r _returnvalue_g _returnvalue_b _returnvalue_a) (Math|Color|BreakColor (Math|Color|Lerp(LinearColor) (Variables|VFX|GetScanningLaserEdgeColor) (Variables|VFX|GetTargetingLaserEdgeColor) _alpha)))
                (Rendering|Material|SetCustomPrimitiveDataVector4 _angleraysstaticmeshcomp 4 (Math|Vector4|MakeVector4 _returnvalue_r _returnvalue_g _returnvalue_b _returnvalue_a))
                (bind (_returnvalue_r_2 _returnvalue_g_2 _returnvalue_b_2 _returnvalue_a_2) (Math|Color|BreakColor (Math|Color|Lerp(LinearColor) (Variables|VFX|GetScanningLaserCenterColor) (Variables|VFX|GetTargetingLaserCenterColor) _alpha)))
                (Rendering|Material|SetCustomPrimitiveDataVector4 _angleraysstaticmeshcomp 8 (Math|Vector4|MakeVector4 _returnvalue_r_2 _returnvalue_g_2 _returnvalue_b_2 _returnvalue_a_2)))))
          (:ThenDo
            (Timeline|TLTargetRayScale Play))))
      (:Backward
        (CallFunction|SetLaserActive false)
        (Utilities|DoOnceThenDo false
          (:DoOnce
            (Timeline|TLTargetRayScale ReverseFromEnd))
          (:ThenDo
            (Timeline|TLTargetRayScale Reverse)))))))

(event Custom|SetTurretHiddenState
  (params (Hidden :type Boolean))
  (bind _tltargetrayscale (Variables|BGA_Athena_Sentry_Turret_OnTrain_Parent|GetTLTargetRayScale))
  (bind _returnvalue (Components|Timeline|IsPlaying _tltargetrayscale))
  (bind _isdormantordeactivated (CallFunction|IsDormantorDeactivated))
  (bind _currentsentrystate (Variables|BuildingGameplayActorSentry|GetCurrentSentryState))
  (bind _returnvalue_2 (Utilities|Enum|Equal(Enum) _currentsentrystate "Aggro"))
  (bind _returnvalue_3 (Utilities|Enum|Equal(Enum) _currentsentrystate "Tracking"))
  (bind _returnvalue_4 (Math|Boolean|ORBoolean _returnvalue _returnvalue_3 _returnvalue_2))
  (bind _angleraysstaticmeshcomp (Variables|BuildingGameplayActorSentry|GetAngleRaysStaticMeshComp))
  ;[0.7187 0.09705 1] Turret open/close animation sequence
  (CallFunction|LogMessage (Utilities|String|Append "SetTurretHiddenState = " (Utilities|String|ToString(Boolean) Hidden)))
  (if Hidden
    (Utilities|DoOnceThenDo false
      (:DoOnce
        (bind (_direction _animation _raylength _rayangle _blades _emissive) (Timeline|TLAnimation PlayFromStart)
          (:Update
            ;;[0 0.7777 1] TLAnimation Update
            (if (not (Variables|HandleMaterialParameters|GetbUsePrimitiveDataForMaterials))
              (Rendering|Material|SetScalarParameterValueonMaterials (Variables|BuildingGameplayActorSentry|GetCameraStaticMeshComp) "Animation" _animation)
              (Rendering|Material|SetScalarParameterValueonMaterials (Variables|Default|GetSM_PG_Turret_Base) "Blades" _blades)
              (if (not _returnvalue_4)
                (Rendering|Material|SetScalarParameterValueonMaterials (Variables|Default|GetSM_PG_Turret_Base) "EmissiveScale" _emissive))
              (else
                (Rendering|Material|SetCustomPrimitiveDataFloat (Variables|BuildingGameplayActorSentry|GetCameraStaticMeshComp) 6 _animation) ; Animation = Index 6 (for Creative)
                (Rendering|Material|SetCustomPrimitiveDataFloat (Variables|Default|GetSM_PG_Turret_Base) 5 _blades) ; Blades = Index 5 (for Creative)
                (if (not _returnvalue_4)
                  (Rendering|Material|SetCustomPrimitiveDataFloat (Variables|Default|GetSM_PG_Turret_Base) 4 _emissive) ; EmissiveScale = Index 4 (for Creative)
                )))
            (Rendering|Material|SetCustomPrimitiveDataFloat _angleraysstaticmeshcomp 3 (Math|Float|Lerp 0.0 (Variables|VFX|GetTargetingLaserFalloff) _raylength))
            (Rendering|Material|SetCustomPrimitiveDataFloat _angleraysstaticmeshcomp 0 (Math|Float|Lerp 0.0 (Variables|SentryParentVars|GetTargetingConeAngle) _rayangle)))
          (:Finished
            ;;[1 0.7324 0.04201] TLAnimation Finished
            (switch Utilities|FlowControl|Switch|SwitchonETimelineDirection _direction
              (:Forward
                (Development|SetHiddeninGame (Variables|BuildingGameplayActorSentry|GetCameraStaticMeshComp) _isdormantordeactivated false))))
          (:SetRayHidden
            ;;[0.5417 0.3212 0.1071] TL Animation Set Ray Hidden
            (Development|SetHiddeninGame (Variables|BuildingGameplayActorSentry|GetAngleRaysStaticMeshComp) _isdormantordeactivated false))))
      (:ThenDo
        (Timeline|TLAnimation Play)))
    (else
      (Development|SetHiddeninGame (Variables|BuildingGameplayActorSentry|GetCameraStaticMeshComp) false false)
      (Utilities|DoOnceThenDo false
        (:DoOnce
          (Timeline|TLAnimation ReverseFromEnd))
        (:ThenDo
          (Timeline|TLAnimation Reverse))))))

(event EventDeactivateRays)

(event EventActivateRays)

(fn CalcAdjustedFireTargetLocation
  (params (TrackedActor :type Actor%20Object%20Reference))
  (bind _trackedactor (Variables|GetTrackedActor)
    (:then
      (bind _asfort_pawn (Utilities|Casting|CastToFortPawn _trackedactor)
        (:then
          (return (Math|Vector|vector+vector (Transformation|GetActorLocation _trackedactor) (Math|Vector|MakeVector 0.0 0.0 (- (Components|Capsule|GetScaledCapsuleHalfHeight (Variables|/Script/Engine.Character|GetCapsuleComponent _asfort_pawn)) 25.0)))))
        (:CastFailed
          (return (Transformation|GetActorLocation _trackedactor)))))
    (:else
      (return (Variables|SentryParentVars|GetLastKnownTargetLocation)))))

(fn CalcTrackedToAggroDelay
  (return (* (ScalableFloat|GetValueatLevel (Variables|BuildingGameplayActorSentry|GetTrackedToAggroDelay)) (Math|Float|Max(Float) (Variables|BuildingGameplayActorSentry|GetTrackedToAggroDelayMultiplier)))))

(fn GetHitEffectLevel
  (return 0.0))

(fn GetStateToSetOnReactivate
  (return (select (<= (Utilities|Array|Length (Variables|BuildingGameplayActorSentry|GetTargetActorsInRange)) 0) (select (Variables|BuildingGameplayActorSentry|GetbGoDormantAfterPassiveIdle) "Dormant" "ReturningToIdle") "ActiveIdle")))

(fn GetWindUpTimer
  (return (Math|Float|Max(Float) (- (CallFunction|CalcTrackedtoAggroDelay) (Variables|Default|GetWindUpSoundDelay)))))

(fn HandleHitWithoutTarget
  (params (HitActor :type Actor%20Object%20Reference))
  ;[1 1 1] We did not hit our target pawn. See if we should stop firing
  (bind _asfort_player_pawn (Utilities|Casting|CastToFortPlayerPawn HitActor)
    (:then
      (return (CallFunction|CanAttackTargetActor _asfort_player_pawn)))
    (:CastFailed
      (return true))))

(fn InitializeAngleRaysParameters
  (bind _angleraysstaticmeshcomp (Variables|BuildingGameplayActorSentry|GetAngleRaysStaticMeshComp))
  (bind (_scanninglaseredgecolor_r _scanninglaseredgecolor_g _scanninglaseredgecolor_b _scanninglaseredgecolor_a) (Math|Color|BreakColor (Variables|GetScanningLaserEdgeColor_R)))
  (bind (_scanninglasercentercolor_r _scanninglasercentercolor_g _scanninglasercentercolor_b _scanninglasercentercolor_a) (Math|Color|BreakColor (Variables|GetScanningLaserCenterColor_R)))
  (Networking|CanExecuteCosmeticEvents
    (:True
      (Rendering|Material|SetCustomPrimitiveDataFloat _angleraysstaticmeshcomp 0 (Variables|SentryParentVars|GetTargetingConeAngle))
      (Rendering|Material|SetCustomPrimitiveDataFloat _angleraysstaticmeshcomp 1 (Variables|VFX|GetScanningLaserEdgeWidthEnd))
      (Rendering|Material|SetCustomPrimitiveDataFloat _angleraysstaticmeshcomp 2 (Variables|VFX|GetScanningLaserEdgeWidthFront))
      (Rendering|Material|SetCustomPrimitiveDataFloat _angleraysstaticmeshcomp 3 (Variables|VFX|GetScanningLaserFalloff))
      (Rendering|Material|SetCustomPrimitiveDataVector4 _angleraysstaticmeshcomp 4 (Math|Vector4|MakeVector4 _scanninglaseredgecolor_r _scanninglaseredgecolor_g _scanninglaseredgecolor_b _scanninglaseredgecolor_a))
      (Rendering|Material|SetCustomPrimitiveDataVector4 _angleraysstaticmeshcomp 8 (Math|Vector4|MakeVector4 _scanninglasercentercolor_r _scanninglasercentercolor_g _scanninglasercentercolor_b _scanninglasercentercolor_a))
      (Rendering|Material|SetCustomPrimitiveDataFloat _angleraysstaticmeshcomp 12 (Variables|VFX|GetSecondaryRaysSpeed))
      (Rendering|Material|SetCustomPrimitiveDataFloat _angleraysstaticmeshcomp 13 (Variables|VFX|GetSecondaryRaysOpacity)))))

(fn IsDormantOrDeactivated
  (return (or (Utilities|Enum|Equal(Enum) (Variables|BuildingGameplayActorSentry|GetCurrentSentryState) "Dormant") (Utilities|Enum|Equal(Enum) (Variables|BuildingGameplayActorSentry|GetCurrentSentryState) "Deactivated"))))

(fn IsFriendlyToLocalPlayer
  (return (Utilities|Enum|Equal(Enum) (Team|GetTeamAffiliationforActortoLocalPlayer self true true) "Friendly")))

(fn OnRep_InSpyLTM
  (if (Networking|IsServer)
    _
    (elif (Variables|Default|GetInSpyLTM)
      ;[1 1 1] Hide the rays on the defender team if in the LTM
      (for _array_element (Fortnite|GetLocalFortPlayerControllers true)
        (if (GameplayTags|TagContainer|CompareTagAssetInterfacetoOtherTags (Variables|Default|GetT_SpyLTMDefender) (Game|Player|GetPlayerPawn _array_element))
          (Rendering|SetVisibility (Variables|BuildingGameplayActorSentry|GetAngleRaysStaticMeshComp) false false)
          (Variables|VFX|SetSeekingColor (Variables|VFX|GetSeekingColorFriendly)))))))

(fn OnRep_IsShooting
  (if (not (Variables|Default|GetIsShooting))
    (Utilities|Time|ClearandInvalidateTimerbyHandle (Variables|Default|GetTimer_Fire))))

(fn ShouldContinueShootingAtTarget
  (params (Actor :type Actor%20Object%20Reference))
  (Utilities|FlowControl|Sequence
    (:then_0
      (Utilities|IsValid Actor
        (:Is%20Valid
          (bind _returnvalue (GameplayTags|HasMatchingGameplayTag(Message) Actor (Variables|Default|GetT_PlayerPawn)))
          (if (not _returnvalue)
            (bind _asbuilding_actor (Utilities|Casting|CastToBuildingActor Actor)
              (:then
                (return (BuildingActor|IsPlayerBuilt _asbuilding_actor)))
              (:CastFailed
                _)))
          (return true))))
    (:then_1
      (return false))))

(fn ConstructionScript
  (bind (_relativerotation_roll _relativerotation_pitch _relativerotation_yaw) (Math|Rotator|BreakRotator (Variables|/Script/Engine.SceneComponent|GetRelativeRotation_Roll (Variables|BuildingGameplayActorSentry|GetCameraStaticMeshComp))))
  (|CallParentFunction)
  (CallFunction|InitializeAngleRaysParameters)
  ;[1 1 1] Save Out Pitch
  (Transformation|SetRelativeRotation (Variables|BuildingGameplayActorSentry|GetCameraStaticMeshComp) (Math|Rotator|MakeRotator _relativerotation_roll (Variables|SentryPublicVars|GetCameraPitch) _relativerotation_yaw) false false))

(fn VFX_SetLightColor)
