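; Destroyed train turret prop that lets a player extract its thermite core once, receiving three throwable Thermite items.
; A short "Extract Thermite Core" interaction calls a multicast-named event that sets ExtractedCore and gives the items to the interacting pawn's controller.
; ExtractedCore blocks further interaction, and its OnRep hides the thermite mesh.
; The construction script zeroes the turret base's "Blades" material parameter.
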(event EventBeginPlay
  (Interaction|BindEventtoOnInteract (delegate Custom|Interacted)))

(event Custom|Interacted_Multicast
  (params (Pawn :type Pawn%20Object%20Reference))
  (Variables|Default|SetExtractedCore true)
  (Inventory|GiveItemtoInventoryOwner :InventoryOwner (Fortnite|GetFortPlayerControllerfromActor Pawn) :ItemDefinition "/ThermiteGameplay/Gameplay/WID_Thermite_Throwable.WID_Thermite_Throwable" :NumberToGive 3 :bNotifyPlayer true :ItemLevel -1 :PickupInstigatorHandle -1 :bUseItemPickupAnalyticEvent false :WeaponAmmoOverride -1))

(event Custom|Interacted
  (params (SelfActor :type Building%20Actor%20Object%20Reference) (InteractingPawn :type Fort%20Pawn%20Object%20Reference))
  (CallFunction|InteractedMulticast InteractingPawn))

(event Collision|EventActorBeginOverlap
  (params (OtherActor :type Actor%20Object%20Reference)))

(event EventTick
  (params (DeltaSeconds :type Float%20%28single-precision%29)))

(fn BlueprintCanInteract
  (params (InteractingPawn :type Fort%20Pawn%20Object%20Reference) (InteractionBeingAttempted :type EInteractionBeingAttempted%20Enum :default "FirstInteraction") (InteractionType :type TInteractionType%20Enum :default "IT_NoInteraction"))
  (return (not (Variables|Default|GetExtractedCore))))

(fn BlueprintGetInteractionString
  (params (InteractingPawn :type Fort%20Pawn%20Object%20Reference) (InteractionBeingAttempted :type EInteractionBeingAttempted%20Enum :default "FirstInteraction"))
  (return "NSLOCTEXT(\"[2F3DB45FADCC4D535EAF09EF8F44C64C]\", \"DA9996674152A834846760A82755B0C6\", \"Extract Thermite Core\")"))

(fn BlueprintGetInteractionTime
  (params (InteractingPawn :type Fort%20Pawn%20Object%20Reference) (InteractionBeingAttempted :type EInteractionBeingAttempted%20Enum :default "FirstInteraction"))
  (return 0.5 true))

(fn OnRep_ExtractedCore
  (Development|SetHiddeninGame (Variables|Default|GetSM_Thermite) (Variables|Default|GetExtractedCore) false))

(fn ConstructionScript
  (Rendering|Material|SetScalarParameterValueonMaterials (Variables|Default|GetSM_PG_Turret_Base1) "Blades" 0.0))
