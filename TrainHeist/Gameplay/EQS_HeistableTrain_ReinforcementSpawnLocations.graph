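# EQS query that finds candidate reinforcement spawn locations around the querier for the Train Heist heistable train.
# A donut generator lays points in a spiral of rings around the querier, projected with a navigation filter that, by its name, excludes smashable objects and water.
# A Project test then filters the points by tracing geometry vertically on a custom game trace channel, with no scoring preference between candidates configured.
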EnvQuery /TrainHeist/Gameplay/EQS_HeistableTrain_ReinforcementSpawnLocations
options [{
  generator {
   name=EnvQueryGenerator_Donut_0
   class=EnvQueryGenerator_Donut
   properties {
    bUseSpiralPattern=true
    center=&EnvQueryContext_Querier
    innerRadius=400
    numberOfRings=4
    outerRadius=1300
    pointsPerRing=10
    projectionData {
     extentX=0
     navigationFilter=&NavFilter_Phoebe_NoSmashableNoWater_C
     postProjectionVerticalOffset=15
     projectDown=512
    }
   }
  }
  tests [EnvQueryTest_Project_0 {
    class=EnvQueryTest_Project
    properties {
     scoringFactor=1
     testComment=""
     testPurpose=Filter
     projectionData {
      navigationFilter=null
      postProjectionVerticalOffset=0
      projectDown=512
      projectUp=1024
      serializedChannel=ECC_GameTraceChannel11
      traceChannel=TraceTypeQuery10
      traceMode=GeometryByChannel
     }
    }
   }]
 }]
