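; Front car of the heistable train, handling engine-destruction stops, minimap boss icon, whistle and audio.
; Once both engines are flagged destroyed (checked only on front-engine destruction), it stops the train via TrainManager; stopping broadcasts OnTrainStopped or spawns a nav-generation actor.
; A timer increments RepNotify WhistleCount to reach clients despite dormancy.
; Impulse targets get launched with gameplay effects, or woken if vehicles; fuse-box counting triggers nothing.
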(event Custom|CheckForPlayersInGap
  (Networking|SwitchHasAuthority
    (:Authority
      (bind _outactors (Collision|SphereOverlapActors (Utilities|Operators|Add (Transformation|GetActorLocation) (* (Transformation|GetActorForwardVector) -1132.0) "0.000000,0.000000,150.000000") 2e+02 (Utilities|Array|MakeArray "ObjectTypeQuery3") "/Script/FortniteGame.FortPawn"))
      (for _array_element _outactors
        (bind _asfort_pawn (Utilities|Casting|CastToFortPawn _array_element))
        (Pawn|Character|LaunchCharacterJump _asfort_pawn (+ (* (Math|Vector|CrossProduct (Math|Vector|Normalize (Transformation|GetVelocity)) "0.000000,0.000000,1.000000") 1e+03) "0.000000,0.000000,1000.000000") true true false true)))))

(event Custom|MarkFuseBoxes
  (params (Player :type Actor%20Object%20Reference))
  (for _array_element (Variables|Default|GetFuse%20Boxes)
    (Utilities|IsValid _array_element
      (:Is%20Valid
        (Development|PrintString (Utilities|GetDisplayName _array_element) true true "(R=1.000000,G=0.000000,B=0.092620,A=1.000000)" 3e+01)
        (MarkerComponent|MarkActoronServer (Class|FortPlayerController|GetMarkerComponent (Fortnite|GetFortPlayerControllerfromActor Player)) _array_element true true)))))

(event Custom|FuseBox%20Destroyed
  (Development|PrintString "FuseBoxDestroyed" true true "(R=1.000000,G=0.028928,B=0.000000,A=1.000000)" 2e+02)
  (bind _result (Math|Integer|IncrementInt (Variables|Default|GetFuseBoxesDestroyed))
    (:%20%20
      (if (>= _result 2)
        _))))

(event Custom|Back%20Engine%20Destroyed
  (Variables|Default|SetBackEngineDestroyed true)
  (Development|PrintString "RECIEVED BACK EVENT" true true "(R=1.000000,G=0.028928,B=0.000000,A=1.000000)" 2e+02))

(event EventBeginPlay
  (bind _returnvalue (CallFunction|GetTrainManager))
  ; Begin Play
  (|CallParentFunction)
  (Networking|SwitchHasAuthority
    (:Authority
      (Utilities|Time|SetTimerbyEvent (EventDispatchers|CreateEvent :!event Custom|Event_NotifyBlowWhistle) 2e+01 true)
      (Class|FortAsyncActionCurrentPlaylistReady|PlaylistDataReadyAsync
        (:OnReady
          (CallFunction|DisplayProperIcon)
          (Utilities|IsValid _returnvalue
            (:Is%20Valid
              (EventDispatchers|BindEventtoOnTrainStoppedEvent _returnvalue (EventDispatchers|CreateEvent :!event Custom|TrainHasStopped)))))))
    (:Remote
      (CallFunction|EnableAudioProcessingforLocalPlayerOnboard 1.0 (Variables|Default|GetFortSoundIndicator) (Variables|Default|GetAudioShapeLineList))
      (EventDispatchers|BindEventtoOnTrainStoppedEvent (CallFunction|GetTrainManager) (EventDispatchers|CreateEvent :!event Custom|TrainStoppedAudio)))))

(event Custom|DisplayProperIcon
  ; Some Logic to check which Map Actor we should be displaying on the train
  (Networking|SwitchHasAuthority
    (:Authority
      (Utilities|FlowControl|Delay 2.0)
      (bind _outactors (Actor|GetAllActorsOfClass "/VividRazor_Minions/Blueprints/BP_VividRazor_BossSpawnLocation_Manager.BP_VividRazor_BossSpawnLocation_Manager_C"))
      (if (> (Utilities|Array|Length _outactors) 0)
        (bind _isbossspawningattrain (Class|BPVividRazorBossSpawnLocationManager|IsBossSupposedtoSpawnatTrain (Utilities|Array|Get(acopy) _outactors)))
        (if _isbossspawningattrain
          (Class|AthenaSpecialActorComponent|SetMiniMapIcon (Variables|Default|GetAthenaSpecialActor) (Widget|Brush|MakeBrushfromMaterial "/TrainHeist/UI/MI_UI_Minimap_Train_WithBoss.MI_UI_Minimap_Train_WithBoss" 40 40)))))))

(event Custom|OnEngineDestroyed_Multicast
  (Variables|Default|SetFrontEngineDestroyed true)
  (Development|PrintString "FRONT DESTROYED" true true "(R=1.000000,G=0.000000,B=0.799272,A=1.000000)" 1e+02)
  (CallFunction|StopifBothDestroyed))

(event Custom|ForceStopTrain
  (bind _returnvalue (CallFunction|GetTrainManager))
  (Networking|SwitchHasAuthority
    (:Authority
      (Utilities|IsValid _returnvalue
        (:Is%20Valid
          (Variables|Default|SetShouldSpawnEnemiesOnStop true)
          (Class|FortTrainManager|BeginStoppingTrain _returnvalue))
        (:Is%20Not%20Valid
          (CallFunction|TrainHasStopped))))
    (:Remote
      (Default|CallOnTrainStopped))))

(event Custom|StopIfBothDestroyed
  (bind _returnvalue (CallFunction|GetTrainManager))
  (if (and (Variables|Default|GetBackEngineDestroyed) (Variables|Default|GetFrontEngineDestroyed))
    (Networking|SwitchHasAuthority
      (:Authority
        (Utilities|IsValid _returnvalue
          (:Is%20Valid
            (Variables|Default|SetShouldSpawnEnemiesOnStop true)
            (Class|FortTrainManager|BeginStoppingTrain _returnvalue))
          (:Is%20Not%20Valid
            (CallFunction|TrainHasStopped))))
      (:Remote
        (Default|CallOnTrainStopped)))))

(event Custom|Engine%20Destroyed
  (CallFunction|OnEngineDestroyedMulticast)
  (Development|PrintString "IT DESTROYED" true true "(R=1.000000,G=0.000000,B=0.799272,A=1.000000)" 1e+02))

(event Custom|TrainHasStopped
  (bind (_returnvalue_location _returnvalue_rotation _returnvalue_scale) (Math|Transform|BreakTransform (Transformation|GetActorTransform self)))
  (Utilities|FlowControl|Sequence
    (:then_0
      ; When Train is stopped, remove the minimap icon (with some grace time)
      (Utilities|FlowControl|Delay 1e+01)
      (Components|Activation|SetActive (Variables|Default|GetAthenaSpecialActor) false false))
    (:then_1
      (if (Variables|Default|GetShouldSpawnEnemiesOnStop)
        (Default|CallOnTrainStopped)
        (Game|SpawnActorfromClass "/TrainHeist/Navigation/BGA_FrontCar_NavGeneration1.BGA_FrontCar_NavGeneration1_C" (Math|Transform|MakeTransform (+ (+ (Transformation|GetActorLocation self) (* (Transformation|GetActorForwardVector self) 4e+02)) (* (Transformation|GetActorUpVector self) 2e+02)) _returnvalue_rotation _returnvalue_scale) "AlwaysSpawn" "MultiplyWithRoot" self 0 0) ; Spawn BGA that will generate NavMesh on top of Train only when the Train is stopped
      ))))

(event EventOnActorTargetedForImpulse
  (params (Actor :type Actor%20Object%20Reference))
  (CallFunction|ApplyFrontImpulse Actor))

(event Custom|Event_NotifyBlowWhistle
  ;[0 0 0] Todo @Seth: drive dynamically on station arrival as well
  (Variables|Default|SetWhistleCount (+ (Variables|Default|GetWhistleCount) 1)) ; Since the train is always dormant, we modify a variable with RepNotify to force a notification to the client
)

(event Custom|Event_BlowWhistle
  ;[0 0 0] Todo @Seth: drive dynamically on station arrival as well
  (Audio|Components|Audio|Play (Variables|Default|GetAudioWhistle) 0.0))

(event Custom|TrainStoppedAudio
  (Audio|Modulation|ActivateControlBusMix "/Train/Sound/Modulation/CBM_Train_Stopped.CBM_Train_Stopped"))

(fn ApplyFrontImpulse
  (params (Actor :type Actor%20Object%20Reference))
  (bind _asfort_pawn (Utilities|Casting|CastToFortPawn Actor)
    (:then
      (Character|LaunchCharacter _asfort_pawn (Math|Vector|MakeVector 0.0 0.0 (ScalableFloat|GetValueatLevel (Variables|Default|GetLaunchSpeed))) true true) ; We need this initial launch since the knockback GE/GA takes too long to activate, and our character could already be on the train. GE/GA still handles knockback status so it handles cancelling other abilities
      (Fortnite|ApplyGameplayEffectsToActor _asfort_pawn (Variables|Default|GetEffects%20to%20Apply) (Transformation|GetActorLocation _asfort_pawn) false 0 (Ability|GetFortAbilitySystemComponent self))
      (GameplayCue|ExecuteGameplayCue _asfort_pawn (Variables|Default|GetPlayerLaunchGameplayCueTag)))
    (:CastFailed
      (bind _asfort_vehicle_interface (Utilities|Casting|CastToFortVehicleInterface Actor))
      (Vehicle|WakeUp _asfort_vehicle_interface))))

(fn OnRep_WhistleCount
  (CallFunction|EventBlowWhistle))

(fn ConstructionScript
  (|CallParentFunction))
