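; Open car Blueprint for the heistable train, handling its turret, fuse box, loot containers and gap safety.
; BeginPlay initializes the child turret, whose death spawns an attached thermite wreck, and after a delay binds the server to train stops, which spawn a rooftop NavMesh generator for AI.
; Fuse box destruction forwards to the front car; container destruction sets RepNotify flags.
; A server-side gap check launches overlapping pawns sideways and upward.
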(event Custom|OnTurretDied
  (params (DamagedActor :type Actor%20Object%20Reference) (Damage :type Float%20%28single-precision%29) (InstigatedBy :type Controller%20Object%20Reference) (DamageCauser :type Actor%20Object%20Reference) (HitLocation :type Vector) (FHitComponent :type Primitive%20Component%20Object%20Reference) (BoneName :type Name) (Momentum :type Vector))
  (bind _returnvalue (Game|SpawnActorfromClass "/TrainHeist/Gameplay/BGA_DestroyedTurretWithThermite_OnTrain.BGA_DestroyedTurretWithThermite_OnTrain_C" (Transformation|GetActorTransform DamagedActor) "Undefined" "MultiplyWithRoot" 0 0 0))
  (Transformation|AttachActorToActor _returnvalue self "None" "KeepWorld" "KeepWorld" "KeepWorld" false))

(event Custom|CheckForPlayersInGap
  (Networking|SwitchHasAuthority
    (:Authority
      (bind _outactors (Collision|SphereOverlapActors (Utilities|Operators|Add (Transformation|GetActorLocation) (* (Transformation|GetActorForwardVector) -1132.0) "0.000000,0.000000,150.000000") 2e+02 (Utilities|Array|MakeArray "ObjectTypeQuery3") "/Script/FortniteGame.FortPawn"))
      (for _array_element _outactors
        (bind _asfort_pawn (Utilities|Casting|CastToFortPawn _array_element))
        (Pawn|Character|LaunchCharacterJump _asfort_pawn (+ (* (Math|Vector|CrossProduct (Math|Vector|Normalize (Transformation|GetVelocity)) "0.000000,0.000000,1.000000") 1e+03) "0.000000,0.000000,1000.000000") true true false true)))))

(event EventBeginPlay
  (bind _returnvalue (CallFunction|GetTrainManager))
  (|CallParentFunction)
  (Utilities|FlowControl|Sequence
    (:then_0
      (bind _asbga_athena_sentry_turret_on_train (Utilities|Casting|CastToBGA_Athena_Sentry_Turret_OnTrain (Variables|/Script/Engine.ChildActorComponent|GetChildActor (Variables|Default|GetTrain_Turret_))))
      (Class|BGAAthenaSentryTurretonTrain|Init _asbga_athena_sentry_turret_on_train self)
      (EventDispatchers|BindEventtoOnDied _asbga_athena_sentry_turret_on_train (delegate Custom|OnTurretDied)))
    (:then_1)
    (:then_3
      (Networking|SwitchHasAuthority
        (:Authority
          (Utilities|FlowControl|Delay 5.0) ; Not sure what's the delay we need here for the Manager to be active. Let's hope this works in a cooked build :D
          (Utilities|IsValid _returnvalue
            (:Is%20Valid
              (EventDispatchers|BindEventtoOnTrainStoppedEvent _returnvalue (EventDispatchers|CreateEvent :!event Custom|TrainStopped)))))))))

(event Custom|FuseBoxDestroyed
  (params (DestroyedActor :type Actor%20Object%20Reference))
  (CallFunction|FuseBoxDestroyedMulticast))

(event Custom|FuseBoxDestroyedMulticast
  (Development|PrintString "BACK DESTROYED" true true "(R=1.000000,G=0.028928,B=0.000000,A=1.000000)" 1e+01)
  (bind _asfront_car_heistable_train (Utilities|Casting|CastToFrontCar_HeistableTrain (Utilities|Struct|BreakTrainCarRuntimeData (Utilities|Array|Get(aref) (Variables|/Script/TrainCoreRuntime.FortTrainManager|GetCars (CallFunction|GetTrainManager))))))
  (Class|FrontCarHeistableTrain|FuseBoxDestroyed _asfront_car_heistable_train))

(event Custom|DestroyLootContainerAudioComponent
  (params (DamagedActor :type Actor%20Object%20Reference) (Damage :type Float%20%28single-precision%29) (InstigatedBy :type Controller%20Object%20Reference) (DamageCauser :type Actor%20Object%20Reference) (HitLocation :type Vector) (FHitComponent :type Primitive%20Component%20Object%20Reference) (BoneName :type Name) (Momentum :type Vector))
  (Variables|Components|SetRepNotifyDestroyContainerOne true))

(event Custom|DestroyLootContainerAudioComponentTwo
  (params (DamagedActor :type Actor%20Object%20Reference) (Damage :type Float%20%28single-precision%29) (InstigatedBy :type Controller%20Object%20Reference) (DamageCauser :type Actor%20Object%20Reference) (HitLocation :type Vector) (FHitComponent :type Primitive%20Component%20Object%20Reference) (BoneName :type Name) (Momentum :type Vector))
  (Variables|Default|SetRepNotifyDestroyContainerTwo true))

(event Custom|TrainStopped
  (bind (_returnvalue_location _returnvalue_rotation _returnvalue_scale) (Math|Transform|BreakTransform (Transformation|GetActorTransform self)))
  ; When the train Stops, Spawn a BGA that will allow NavMesh to be built on top of the train and allow AI to jump on top and off the train.
  (Game|SpawnActorfromClass "/TrainHeist/Navigation/BGA_StandardCar_NavGeneration.BGA_StandardCar_NavGeneration_C" (Math|Transform|MakeTransform (+ (+ (Transformation|GetActorLocation self) (* (Transformation|GetActorForwardVector self) 292.0)) (* (Transformation|GetActorUpVector self) 175.0)) _returnvalue_rotation _returnvalue_scale) "AlwaysSpawn" "MultiplyWithRoot" self 0 0) ; Spawn BGA that will generate NavMesh on top of Train only when the Train is stopped
)

(fn ConstructionScript
  (|CallParentFunction))
