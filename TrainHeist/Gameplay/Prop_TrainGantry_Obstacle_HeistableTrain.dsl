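; Train gantry obstacle prop for the heistable train that knocks train-riding Fortnite pawns away when they collide with it.
; BeginPlay calls the parent and binds an actor-hit handler, which acts only on pawns tagged OnTrain and not Falling.
; Launch direction is the obstacle-to-pawn offset with the gantry's forward component removed, normalized in 2D, giving a strong sideways push plus upward boost.
; A debug "Launch!" string prints afterward; overlap and tick events are empty.
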(event Collision|EventActorBeginOverlap
  (params (OtherActor :type Actor%20Object%20Reference)))

(event EventTick
  (params (DeltaSeconds :type Float%20%28single-precision%29)))

(event EventBeginPlay
  (|CallParentFunction)
  (Collision|BindEventtoOnActorHit (delegate Custom|OnActorHit_Event)))

(event Custom|OnActorHit_Event
  (params (SelfActor :type Actor%20Object%20Reference) (OtherActor :type Actor%20Object%20Reference) (NormalImpulse :type Vector) (Hit :type Hit%20Result%20Structure%20%28by%20ref%29))
  (bind _asfort_pawn (Utilities|Casting|CastToFortPawn OtherActor))
  (if (and (GameplayTags|HasMatchingGameplayTag _asfort_pawn "(TagName=\"Gameplay.Action.Pawn.OnTrain\")") (not (GameplayTags|HasMatchingGameplayTag _asfort_pawn "(TagName=\"MovementMode.Falling\")")))
    (Utilities|LaunchPawns _asfort_pawn _asfort_pawn (Utilities|Operators|Add :A (* (Math|Vector|Normalize2D(Vector) (Math|Vector|ProjectVectorontoPlane (- (Transformation|GetActorLocation OtherActor) (Transformation|GetActorLocation SelfActor)) (Transformation|GetActorForwardVector))) 1.9e+03) :B "0.000000,0.000000,1750.000000") true true false true true)
    (Development|PrintString "Launch!")))

(fn ConstructionScript)
