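; Thermite explosive prop on the heistable train's vault car that players place to blow the vault open.
; Interacting with thermite starts a fuse with fire, alarm and timed lock-in, while the train stopping or health depletion starts a delayed detonation.
; Detonation opens the vault, then death runs a throttled async overlap with line-of-sight checks to damage and launch actors.
; RepNotifies drive explosion cues, and StopDetonation resets the fuse and submits analytics.
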(event EventBeginPlay
  (|CallParentFunction)
  (CallFunction|OpenTop)
  (Variables|Default|SetMI_Thermite_0 (Rendering|Material|GetMaterial (Variables|Debug|GetStaticMeshComponent) 0))
  (CallFunction|UpdateThermiteMaterials false))

(event Custom|Init
  (params (CarPropIsAttachedTo :type Base%20Car%20Object%20Reference))
  (Utilities|FlowControl|Delay 3.0) ; Make sure all the cars have spawned
  (bind _output_get (Variables|Default|SetTrainManager (Class|TrainCar|GetTrainManager CarPropIsAttachedTo)))
  (for _array_element (Variables|/Script/TrainCoreRuntime.FortTrainManager|GetCars _output_get)
    (bind _asvault_car_heistable_train (Utilities|Casting|CastToVaultCar_HeistableTrain (Utilities|Struct|BreakTrainCarRuntimeData _array_element)))
    (Variables|Default|SetVaultCar _asvault_car_heistable_train))
  (EventDispatchers|BindEventtoOnTrainStoppedEvent _output_get (delegate Custom|OnTrainStopped))
  (EventDispatchers|BindEventtoOnDeathPlayEffectsDelegate (delegate Custom|PlayDeathAudio)))

(event Custom|OnTrainStopped
  (CallFunction|StartDetonation))

(event Custom|PlayDeathAudio
  (params (Damage :type Float%20%28single-precision%29) (DamageTags :type Gameplay%20Tag%20Container%20Structure%20%28by%20ref%29) (Momentum :type Vector) (HitInfo :type Hit%20Result%20Structure%20%28by%20ref%29) (InstigatedBy :type Fort%20Pawn%20Object%20Reference) (DamageCauser :type Actor%20Object%20Reference) (EffectContext :type Gameplay%20Effect%20Context%20Handle%20Structure))
  (Networking|CanExecuteCosmeticEvents
    (:True
      (Audio|PlaySoundatLocation "/ThermiteGameplay/Assets/Sounds/Train/SC_Thermite_Train_Explosion.SC_Thermite_Train_Explosion" (Transformation|GetActorLocation)))))

(event Custom|OpenTop
  (Variables|Default|SetIsTopOpen true))

(event Interaction|EventBlueprintOnInteract
  (params (InteractingPawn :type Fort%20Pawn%20Object%20Reference) (InteractionBeingAttempted :type EInteractionBeingAttempted%20Enum :default "FirstInteraction"))
  (bind _returnvalue (Fortnite|GetFortPlayerControllerfromActor InteractingPawn))
  (|CallParentFunction InteractingPawn InteractionBeingAttempted)
  (bind _retvalue (CallFunction|DoesPlayerHaveThermite InteractingPawn)) ; Check if player has thermite
  (if _retvalue
    (Inventory|RemoveItemFromPlayer :PlayerController _returnvalue :ItemDefinition (Variables|Default|GetThermiteItem) :AmountToRemove 1 :bForceRemoval true :bSkipVerbRemoval false)
    (Inventory|RemoveItemFromPlayer :PlayerController _returnvalue :ItemDefinition (Variables|Default|GetThermiteItem_Keycard) :AmountToRemove 1 :bForceRemoval true :bSkipVerbRemoval false) ; Support keycard version for now
    (if (Variables|Default|GetIsTopOpen)
      (CallFunction|StartFuse _returnvalue InteractingPawn))))

(event Custom|StartDetonation_Multicast
  (CallFunction|StartDetonation))

(event Interaction|EventBlueprintOnBeginInteract
  (params (InteractingPawn :type Fort%20Pawn%20Object%20Reference) (InteractionBeingAttempted :type EInteractionBeingAttempted%20Enum :default "FirstInteraction") (InteractionType :type TInteractionType%20Enum :default "IT_NoInteraction"))
  (CallFunction|OnStartedInteractServer InteractingPawn)
  (Default|CallOnStartedInteracting InteractingPawn))

(event Custom|OnStartedInteract_Server
  (params (InteractingPawn :type Fort%20Pawn%20Object%20Reference))
  (Default|CallOnStartedInteracting InteractingPawn))

(event Custom|ExplodeAndDieStaggered
  ;[1 1 1] Damge Logic: If Destroyed by Anything but PickAxe, Explode
  ;;[1 0.4346 0] Special Case for handling chain reaction of explosions, to stagger them to ensure we do not simply delete a player than is too close (Jason West Request)
  (CallFunction|GrowTimeline)
  (Rendering|SetActorHiddenInGame true)
  (Variables|Default|SetbShouldDie true))

(event Custom|TriggeredByGasPump
  ;[1 1 1] Damge Logic: If Destroyed by Anything but PickAxe, Explode
  ;;[1 0.4346 0] Special Case for handling chain reaction of explosions, to stagger them to ensure we do not simply delete a player than is too close (Jason West Request)
  (Utilities|Time|SetTimerbyEvent (delegate Custom|ExplodeAndDieStaggered) (Math|Random|RandomFloatinRange 1.7 4.5))
  (CallFunction|StartFuse 0))

(event EventOnDamageServer
  (params (Damage :type Float%20%28single-precision%29) (DamageTags :type Gameplay%20Tag%20Container%20Structure%20%28by%20ref%29) (Momentum :type Vector :default "0, 0, 0") (HitInfo :type Hit%20Result%20Structure%20%28by%20ref%29) (InstigatedBy :type Controller%20Object%20Reference) (DamageCauser :type Actor%20Object%20Reference) (EffectContext :type Gameplay%20Effect%20Context%20Handle%20Structure))
  ;[1 1 1] Damge Logic: If Destroyed by Anything but PickAxe, Explode
  (bind _asfort_player_controller (Utilities|Casting|CastToFortPlayerController InstigatedBy)
    (:then
      (Variables|Default|SetLastDamageInstigator _asfort_player_controller))
    (:CastFailed
      _))
  (Utilities|Casting|CastToApollo_GasPump_Valet DamageCauser
    (:then
      (Utilities|FlowControl|Sequence
        (:then_0
          (CallFunction|TriggeredbyGasPump))
        (:then_1
          _)))
    (:CastFailed
      (if (and (>= Damage 1e+02) (GameplayTags|HasAnyTags DamageTags (Variables|Default|GetExplosiveDamageTag) true))
        ;;[1 0.3551 0] If the damage was Explosive and dealt more than 100 damage, we kill much faster than normal
        (Utilities|Time|SetTimerbyEvent (delegate Custom|FastExplodeAndDie) 0.25)
        (else
          _))))
  (Utilities|FlowControl|Gate false
    (:Exit
      (Utilities|FlowControl|Sequence
        (:then_0
          (if (Variables|Bools|GetHealthIsZero)
            (CallFunction|StartDetonation)))
        (:then_1
          ;;[1 1 1] Set On Fire
          (if (<= (Health|GetHealthPercent) (Variables|Default|GetOnFireThreshold))
            (CallFunction|StartFuse 0)))))))

(event Custom|FastExplodeAndDie
  ;[1 1 1] Damge Logic: If Destroyed by Anything but PickAxe, Explode
  ;;[1 0.3551 0] If the damage was Explosive and dealt more than 100 damage, we kill much faster than normal
  (Development|SetHiddeninGame (Variables|Debug|GetStaticMeshComponent) true false)
  (Gameplay|Die "(TagName=\"Gameplay.Effect.InstantDeath\")" (stale EventOnDamageServer InstigatedBy) (stale EventOnDamageServer DamageCauser)) ; DestroyActor credits Self with destruction and \nresulting explosion for quest purposes. \nTo preserve instigator use Die Instead.
  (Development|PrintString "ssa dDESTROYED" true true "(R=1.000000,G=0.000000,B=0.664457,A=1.000000)" 2e+02))

(event Custom|StartFuse
  (params (Killer :type Controller%20Object%20Reference) (DamageCauser :type Actor%20Object%20Reference))
  ;[0.4792 0.2571 0] Detonation Logic:
  (Utilities|FlowControl|DoOnce false
    (:Completed
      (CallFunction|UpdateThermiteMaterials true)
      (Variables|Default|SetFuseKiller Killer)
      (Variables|Default|SetFuseDamageCauser DamageCauser)
      (Variables|Default|SetFuseStarted true)
      (Variables|Analytics|SetStartTime (Fortnite|GetRealMatchTime))
      (Variables|Analytics|SetAnalyticsSubmitted false)
      (Ability|SendGameplayEventtoActor (Variables|Default|GetFuseKiller) "(TagName=\"Event.VividRazor.Thermite.Attached.Vault\")" (Utilities|Struct|MakeGameplayEventData "(TagName=\"Event.VividRazor.Thermite.Attached.Vault\")" self))
      (if (CallFunction|CanbeDamaged 0)
        (Utilities|IsValid (Variables|Default|GetLastDamageInstigator)
          (:Is%20Valid
            (GameplayEffects|FortApplyGameplayEffectToTarget :self (Ability|GetFortAbilitySystemComponent (Pawn|GetControlledPawn (Variables|Default|GetLastDamageInstigator))) :GameplayEffectClass "/Game/Athena/Items/EnvironmentalItems/ReactiveProps/GE_Athena_ReactiveProps_OnFire.GE_Athena_ReactiveProps_OnFire_C" :TargetOfEffect (Ability|GetFortAbilitySystemComponent self) :OptionalSourceObject (Pawn|GetControlledPawn (Variables|Default|GetLastDamageInstigator)) :Level 0.0))
          (:Is%20Not%20Valid
            (Environment|ApplyGlobalEnvironmentGameplayEffecttoActor self (Variables|Default|GetEnvironmentalOnFireGE))))
        (else
          _))
      (CallFunction|OnFireFx)
      (bind _returnvalue (Utilities|Time|SetTimerbyEvent (EventDispatchers|CreateEvent :!event Custom|LockInDetonation) 4.0))
      (Variables|Default|SetLockInDetonationTimerHandle _returnvalue)
      (Default|CallOnFuseStarted)
      (Development|PrintString "FUSE STARTED" true false "(R=1.000000,G=0.006171,B=0.000000,A=1.000000)" 5e+03))))

(event Custom|StopDetonation
  ;[0.4792 0.2571 0] Detonation Logic:
  (CallFunction|ClearDetonationTImers)
  (if (CallFunction|CanbeDamaged 0)
    (Environment|RemoveGlobalEnvironmentGameplayEffect (Variables|Default|GetEnvironmentalOnFireGE)))
  (Variables|Default|SetFuseStarted false)
  (CallFunction|RemoveFireFX)
  (CallFunction|StopGrowAudio)
  (CallFunction|ResetFuse)
  (CallFunction|SetThermitetoHologram)
  (Default|CallOnStopDetonation)
  (Analytics|SubmitAnalytics 0 false)
  (if (ScalableFloat|GetValueasBool (Variables|Default|GetShowTutorialTextWhenDestroyedByTurret))
    (Networking|SwitchHasAuthority
      (:Authority
        (CallFunction|ShowFailedMessage)))))

(event Custom|ShowFailedMessage
  ;[0.4792 0.2571 0] Detonation Logic:
  (bind _asfort_player_pawn (Utilities|Casting|CastToFortPlayerPawn (Variables|Default|GetFuseDamageCauser)))
  (FailedAction|UnabletoPerformAction _asfort_player_pawn false (GameplayTags|MakeLiteralGameplayTagContainer "(GameplayTags=((TagName=\"Activate.Failed.TrainHeist.PlacedThermiteBeforeDestroyingTurrets\")))")))

(event Custom|ClearDetonationTImers
  ;[0.4792 0.2571 0] Detonation Logic:
  (Utilities|Time|ClearandInvalidateTimerbyHandle (Variables|Default|GetExplodeAndDieTimerHandle))
  (Utilities|Time|ClearandInvalidateTimerbyHandle (Variables|Default|GetLockInDetonationTimerHandle))
  (Utilities|Time|ClearandInvalidateTimerbyHandle (Variables|Default|GetFuseTimerHandle)))

(event Custom|LockInDetonation
  ;[0.4792 0.2571 0] Detonation Logic:
  (Utilities|FlowControl|DoOnce false
    (:Completed
      (Default|CallOnDetonationLockedIn)
      (Development|PrintString "DETONATION LOCKED IN" true false "(R=1.000000,G=0.006171,B=0.000000,A=1.000000)" 5e+03)
      (Networking|SwitchHasAuthority
        (:Authority
          (Variables|Default|SetFirstExplosiveRepNotify true))))))

(event Custom|ResetFuse
  ;[0.4792 0.2571 0] Detonation Logic:
  (Utilities|FlowControl|Sequence
    (:then_0
      (Utilities|FlowControl|DoOnce false
        (:Completed
          (CallFunction|UpdateThermiteMaterials true)
          (Variables|Default|SetFuseKiller (stale Custom|StartFuse Killer))
          (Variables|Default|SetFuseDamageCauser (stale Custom|StartFuse DamageCauser))
          (Variables|Default|SetFuseStarted true)
          (Variables|Analytics|SetStartTime (Fortnite|GetRealMatchTime))
          (Variables|Analytics|SetAnalyticsSubmitted false)
          (Ability|SendGameplayEventtoActor (Variables|Default|GetFuseKiller) "(TagName=\"Event.VividRazor.Thermite.Attached.Vault\")" (Utilities|Struct|MakeGameplayEventData "(TagName=\"Event.VividRazor.Thermite.Attached.Vault\")" self))
          (if (CallFunction|CanbeDamaged 0)
            (Utilities|IsValid (Variables|Default|GetLastDamageInstigator)
              (:Is%20Valid
                (GameplayEffects|FortApplyGameplayEffectToTarget :self (Ability|GetFortAbilitySystemComponent (Pawn|GetControlledPawn (Variables|Default|GetLastDamageInstigator))) :GameplayEffectClass "/Game/Athena/Items/EnvironmentalItems/ReactiveProps/GE_Athena_ReactiveProps_OnFire.GE_Athena_ReactiveProps_OnFire_C" :TargetOfEffect (Ability|GetFortAbilitySystemComponent self) :OptionalSourceObject (Pawn|GetControlledPawn (Variables|Default|GetLastDamageInstigator)) :Level 0.0))
              (:Is%20Not%20Valid
                (Environment|ApplyGlobalEnvironmentGameplayEffecttoActor self (Variables|Default|GetEnvironmentalOnFireGE))))
            (else
              _))
          (CallFunction|OnFireFx)
          (bind _returnvalue (Utilities|Time|SetTimerbyEvent (EventDispatchers|CreateEvent :!event Custom|LockInDetonation) 4.0))
          (Variables|Default|SetLockInDetonationTimerHandle _returnvalue)
          (Default|CallOnFuseStarted)
          (Development|PrintString "FUSE STARTED" true false "(R=1.000000,G=0.006171,B=0.000000,A=1.000000)" 5e+03))))
    (:then_1
      (Utilities|FlowControl|DoOnce false
        (:Completed
          (Default|CallOnDetonationLockedIn)
          (Development|PrintString "DETONATION LOCKED IN" true false "(R=1.000000,G=0.006171,B=0.000000,A=1.000000)" 5e+03)
          (Networking|SwitchHasAuthority
            (:Authority
              (Variables|Default|SetFirstExplosiveRepNotify true))))))
    (:then_2
      (Utilities|FlowControl|DoOnce false
        (:Completed
          (CallFunction|GrowTimeline)
          (Variables|Default|SetCanGenerateResources false)
          (bind _returnvalue_2 (Utilities|Time|SetTimerbyEvent (delegate Custom|ExplodeAndDieTrain) (ScalableFloat|GetValueatLevel (Variables|Default|GetSF_DelayBeforeExplosion))))
          (Variables|Default|SetExplodeAndDieTimerHandle _returnvalue_2)
          (Development|PrintString "DETONATION STARTED" true false "(R=1.000000,G=0.006171,B=0.000000,A=1.000000)" 5e+03))))
    (:then_3
      (CallFunction|PlayAlarmAudio false))))

(event Custom|OnFireFx
  ;[0.4792 0.2571 0] Detonation Logic:
  (Projectile|AddGameplayCue%28Looping%29Local self (Variables|Default|GetFireCue) (Ability|GameplayCue|MakeGameplayCueParameters 0.0))
  (CallFunction|PlayAlarmAudio true))

(event Custom|StartDetonation
  ;[0.4792 0.2571 0] Detonation Logic:
  (Utilities|FlowControl|DoOnce false
    (:Completed
      (CallFunction|GrowTimeline)
      (Variables|Default|SetCanGenerateResources false)
      (bind _returnvalue (Utilities|Time|SetTimerbyEvent (delegate Custom|ExplodeAndDieTrain) (ScalableFloat|GetValueatLevel (Variables|Default|GetSF_DelayBeforeExplosion))))
      (Variables|Default|SetExplodeAndDieTimerHandle _returnvalue)
      (Development|PrintString "DETONATION STARTED" true false "(R=1.000000,G=0.006171,B=0.000000,A=1.000000)" 5e+03))))

(event Custom|RemoveFireFX
  ;[0.4792 0.2571 0] Detonation Logic:
  (Projectile|RemoveGameplayCue%28Looping%29Local self (Variables|Default|GetFireCue) (Ability|GameplayCue|MakeGameplayCueParameters 0.0)))

(event Custom|ExplodeAndDieTrain
  ;[0.4792 0.2571 0] Detonation Logic:
  ;; Probably dies before it can send this to the client in most cases, but the server goes through fine.
  (Default|CallOnDetonation)
  (Development|SetHiddeninGame (Variables|Debug|GetStaticMeshComponent) true false)
  (Class|VaultCarHeistableTrain|OpenValutMultiCast (Variables|Default|GetVaultCar))
  (Analytics|SubmitAnalytics 0 true)
  (Ability|SendGameplayEventtoActor (Variables|Default|GetFuseKiller) "(TagName=\"Event.VividRazor.Vault.Opened\")" (Utilities|Struct|MakeGameplayEventData "(TagName=\"Event.VividRazor.Vault.Opened\")" self 0 0 0 "()" (Tags|GetGameplayTags)))
  (Gameplay|Die "(TagName=\"Gameplay.Effect.InstantDeath\")" (Variables|Default|GetFuseKiller) (Variables|Default|GetFuseDamageCauser)) ; DestroyActor credits Self with destruction and \nresulting explosion for quest purposes. \nTo preserve instigator use Die Instead.
)

(event Custom|SetThermiteToHologram
  ;[0.4792 0.2571 0] Detonation Logic:
  (CallFunction|UpdateThermiteMaterials false))

(event Custom|ExplodeAndDie
  ;[0.4792 0.2571 0] Detonation Logic:
  (Health|SetHealth 0.0)
  (Rendering|SetActorHiddenInGame true)
  (Variables|Default|SetbShouldDie true)
  (Development|PrintString "DIEEEEEE" true true "(R=0.369605,G=1.000000,B=0.000000,A=1.000000)" 238.0)
  (Default|CallOnDetonation))

(event Custom|GrowTimeline
  ;[1 1 1] Pre Explode Grow Timeline
  (Networking|CanExecuteCosmeticEvents
    (:True
      (Audio|Components|Audio|Play (Variables|Default|GetPreExplodeAudioComponent) 0.0))))

(event Custom|ExplosionTrace
  ;[1 0.2249 0.04801] Explosive Damage Logic
  (Utilities|FlowControl|DoOnce false
    (:Completed
      (bind _returnvalue (Utility|OverlapQueue|GetAsyncOverlapQueue self))
      (Class|AsyncOverlapQueue|ThrottledAsyncSphereOverlapActors :self _returnvalue :Location (+ (Transformation|GetActorLocation) (Variables|Default|GetExplosionHeightOffset)) :Radius (ScalableFloat|GetValueatLevel (Variables|Default|GetRadius)) :bCheckLineOfSight false :ShouldCancelProcess (EventDispatchers|CreateEvent :!event ShouldCancelExplosion) :ProcessHitActorFunction (delegate Custom|DestroyHitActors) :CompletionFunction (delegate Custom|OnCompletion) :ObjectTypes "ObjectTypeQuery1" :TraceChannel "ECC_Visibility"))))

(event Custom|StopGrowAudio
  ;[1 1 1] Pre Explode Grow Timeline
  (Networking|CanExecuteCosmeticEvents
    (:True
      (bind _preexplodeaudiocomponent (Variables|Default|GetPreExplodeAudioComponent)
        (:then
          (if (Audio|Components|Audio|IsPlaying _preexplodeaudiocomponent)
            (Audio|Components|Audio|FadeOut _preexplodeaudiocomponent 0.25 0.0 "Linear" :!class /Script/Engine.AudioComponent)))
        (:else)))))

(event Custom|DestroyHitActors
  (params (HitActor :type Actor%20Object%20Reference))
  ;[1 0.2249 0.04801] Explosive Damage Logic
  (bind _in_line_of_sight (Class|MissionBlueprintFunctionLibrary|CheckBasicActorLineOfSight self (Variables|Default|GetExplosionTraceOffset) HitActor "0, 0, 0" false "TraceTypeQuery5" "Persistent" 15.0 true))
  (if _in_line_of_sight
    (CallFunction|ExecuteExplosiononActor HitActor)))

(event Custom|OnCompletion
  (params (WasCancelled :type Boolean)))

(event Custom|PlayAlarmAudio
  (params (bPlay :type Boolean))
  ; Alarm Audio
  (Networking|CanExecuteCosmeticEvents
    (:True
      (if bPlay
        (if (not (SoundLibrary|IsEventActive (Variables|Default|GetSoundLibraryComponent_VividRazorVault) (Variables|Default|GetAlarmAudioTag)))
          (SoundLibrary|PlayEvent (Variables|Default|GetSoundLibraryComponent_VividRazorVault) (Variables|Default|GetAlarmAudioTag)))
        (else
          (SoundLibrary|StopEvent (Variables|Default|GetSoundLibraryComponent_VividRazorVault) (Variables|Default|GetAlarmAudioTag)))))))

(event EventOnDeathServer
  (params (Damage :type Float%20%28single-precision%29) (DamageTags :type Gameplay%20Tag%20Container%20Structure%20%28by%20ref%29) (Momentum :type Vector :default "0, 0, 0") (HitInfo :type Hit%20Result%20Structure%20%28by%20ref%29) (InstigatedBy :type Controller%20Object%20Reference) (DamageCauser :type Actor%20Object%20Reference) (EffectContext :type Gameplay%20Effect%20Context%20Handle%20Structure))
  ;[1 1 1] On Death Logic
  (bind _returnvalue (MissionHelpers|HasTags DamageTags (GameplayTags|MakeGameplayTagContainerfromArray (Utilities|Array|MakeArray "(TagName=\"Gameplay.Effect.InstantDeath.Silent\")" "(TagName=\"Gameplay.Effect.InstantDeath.StructuralSupport\")" "(TagName=\"Gameplay.Effect.InstantDeath.PersistentUpdate\")" "(TagName=\"Gameplay.Effect.InstantDeath.Environment\")" "(TagName=\"Gameplay.Effect.InstantDeath.Detachment\")" "(TagName=\"Gameplay.Effect.InstantDeath.Despawn\")" "(TagName=\"Gameplay.Effect.InstantDeath.Culled\")")) false false))
  (if (not _returnvalue)
    (CallFunction|ExplosionTrace)
    ;;[1 1 1] This is because a reliable multicast wasn't getting called everytime. Todd Eckert, Dalvin Kang.
    (Variables|Default|SetExplosiveFxRepNotify true)))

(event EventOnDeathPlayEffects
  (params (Damage :type Float%20%28single-precision%29) (DamageTags :type Gameplay%20Tag%20Container%20Structure%20%28by%20ref%29) (Momentum :type Vector :default "0, 0, 0") (HitInfo :type Hit%20Result%20Structure%20%28by%20ref%29) (InstigatedBy :type Fort%20Pawn%20Object%20Reference) (DamageCauser :type Actor%20Object%20Reference) (EffectContext :type Gameplay%20Effect%20Context%20Handle%20Structure))
  (CallFunction|StopGrowAudio)
  (CallFunction|PlayAlarmAudio false))

(event Custom|DeactivateEmitter
  ;[1 1 1] Deactivate attached spawned emitters on death (to avoid having them linger for a second, they destroy automatically)
  (if (not (Networking|IsServer))
    (Projectile|RemoveGameplayCue%28Looping%29Local self (Variables|Default|GetFireCue) (Ability|GameplayCue|MakeGameplayCueParameters 0.0))
    (for _array_element (Variables|DamageBasedEmitters|GetSpawnedDamageStateEmitters)
      (Components|Activation|Deactivate _array_element))))

(fn ApplyDamageToActor
  (params (Actor%20to%20Damage :type Actor%20Object%20Reference))
  (bind _lastdamageinstigator (Variables|Default|GetLastDamageInstigator))
  (Utilities|IsValid _lastdamageinstigator
    (:Is%20Valid
      (GameplayEffects|FortApplyGameplayEffectToTarget :self (Ability|GetFortAbilitySystemComponent (Pawn|GetControlledPawn _lastdamageinstigator)) :GameplayEffectClass (Variables|GameplayEffects|GetGE_Explode) :TargetOfEffect (Ability|GetFortAbilitySystemComponent (Variables|GetActor%20to%20Damage)) :OptionalSourceObject self :OptionalEffectCauser self :OptionalDamageSource self :Level 0.0))
    (:Is%20Not%20Valid
      (bind _asfort_pawn (Utilities|Casting|CastToFortPawn (Variables|GetActor%20to%20Damage))
        (:then
          (Utilities|IsinVehicle _asfort_pawn
            (:false
              _)))
        (:CastFailed
          (Utilities|Casting|CastToFortVehicleInterface (Variables|GetActor%20to%20Damage)
            (:then
              (Environment|ApplyGlobalEnvironmentDamagetoActor (Variables|GetActor%20to%20Damage) (ScalableFloat|GetValueatLevel (Variables|Default|GetVehicleDamage))))
            (:CastFailed
              _))))
      (Environment|ApplyGlobalEnvironmentGameplayEffecttoActor (Variables|GetActor%20to%20Damage) (Variables|GameplayEffects|GetGE_Explode)))))

(fn CanAttemptGenerateResources
  (params (InTags :type Gameplay%20Tag%20Container%20Structure%20%28by%20ref%29) (InstigatorController :type Controller%20Object%20Reference))
  (return (Variables|Default|GetCanGenerateResources)))

(fn BlueprintCanInteract
  (params (InteractingPawn :type Fort%20Pawn%20Object%20Reference) (InteractionBeingAttempted :type EInteractionBeingAttempted%20Enum :default "FirstInteraction") (InteractionType :type TInteractionType%20Enum :default "IT_NoInteraction"))
  (bind _returnvalue (Fortnite|GetFortPlayerControllerfromActor InteractingPawn))
  (if (not (Variables|Default|GetFuseStarted))
    (Utilities|IsValid _returnvalue
      (:Is%20Valid
        (Utilities|IsValid (Inventory|FindExistingItemForDefinition :self _returnvalue :ItemDefinition (Variables|Default|GetThermiteItem) :bInStorageVault false)
          (:Is%20Valid
            (return true))
          (:Is%20Not%20Valid
            (Utilities|IsValid (Inventory|FindExistingItemForDefinition :self _returnvalue :ItemDefinition (Variables|Default|GetThermiteItem_Keycard) :bInStorageVault false)
              (:Is%20Valid
                (return true))
              (:Is%20Not%20Valid
                (return false)))))))
    (else
      (return false))))

(fn BlueprintGetFailedInteractionString
  (params (InteractingPawn :type Fort%20Pawn%20Object%20Reference) (InteractionBeingAttempted :type EInteractionBeingAttempted%20Enum :default "FirstInteraction"))
  (if (Variables|Default|GetFuseStarted)
    (return "")
    (else
      (return "NSLOCTEXT(\"[F026318441062D0B3F7F6EBE013192DB]\", \"5A19ECCF4FF62BDD9E3B2DAE01532673\", \"Requires Thermite\")"))))

(fn BlueprintGetInteractionString
  (params (InteractingPawn :type Fort%20Pawn%20Object%20Reference) (InteractionBeingAttempted :type EInteractionBeingAttempted%20Enum :default "FirstInteraction"))
  (bind _returnvalue (Fortnite|GetFortPlayerControllerfromActor InteractingPawn))
  (Utilities|FlowControl|Sequence
    (:then_0
      (Utilities|IsValid _returnvalue
        (:Is%20Valid
          (Utilities|IsValid (Inventory|FindExistingItemForDefinition :self _returnvalue :ItemDefinition (Variables|Default|GetThermiteItem) :bInStorageVault false)
            (:Is%20Valid
              (return "NSLOCTEXT(\"[F026318441062D0B3F7F6EBE013192DB]\", \"0B58D9FC42EB6734FFEEE1B00B9055BD\", \"Place Thermite\")"))
            (:Is%20Not%20Valid
              (Utilities|IsValid (Inventory|FindExistingItemForDefinition :self _returnvalue :ItemDefinition (Variables|Default|GetThermiteItem_Keycard) :bInStorageVault false)
                (:Is%20Valid
                  (return "NSLOCTEXT(\"[F026318441062D0B3F7F6EBE013192DB]\", \"7180D44B42BE1E27CE9A58A4969A4583\", \"Place Thermite\")"))))))))
    (:then_1
      (return "NSLOCTEXT(\"[F026318441062D0B3F7F6EBE013192DB]\", \"DC32B0154195465788E4B0858B7896F6\", \"Requires Thermite\")"))))

(fn BlueprintGetInteractionTime
  (params (InteractingPawn :type Fort%20Pawn%20Object%20Reference) (InteractionBeingAttempted :type EInteractionBeingAttempted%20Enum :default "FirstInteraction"))
  (return (Variables|Default|GetInteractionTime) true))

(fn BlueprintModifyIncomingDamage
  (params (Damage :type Float%20%28single-precision%29) (InTags :type Gameplay%20Tag%20Container%20Structure%20%28by%20ref%29) (EffectContext :type Gameplay%20Effect%20Context%20Handle%20Structure%20%28by%20ref%29) (EventInstigator :type Controller%20Object%20Reference) (DamageCauser :type Actor%20Object%20Reference))
  ;[1 1 1] Clamp incoming damage to Current Health - 1 to prevent visual death/destruction/hiding before explosion
  (return (Math|Float|SelectFloat Damage (- (Health|GetHealth) 1.0) (< Damage (Health|GetHealth)))))

(fn DoesPlayerHaveThermite
  (params (PlayerPawn :type Fort%20Pawn%20Object%20Reference))
  (bind _returnvalue (Fortnite|GetFortPlayerControllerfromActor PlayerPawn))
  (Utilities|FlowControl|Sequence
    (:then_0
      (Utilities|IsValid _returnvalue
        (:Is%20Valid
          (Utilities|IsValid (Inventory|FindExistingItemForDefinition :self _returnvalue :ItemDefinition (Variables|Default|GetThermiteItem) :bInStorageVault false)
            (:Is%20Valid
              (return true))
            (:Is%20Not%20Valid
              (Utilities|IsValid (Inventory|FindExistingItemForDefinition :self _returnvalue :ItemDefinition (Variables|Default|GetThermiteItem_Keycard) :bInStorageVault false)
                (:Is%20Valid
                  (return false))))))))
    (:then_1
      (return false))))

(fn ExecuteExplosionOnActor
  (params (ActorToExplode :type Actor%20Object%20Reference))
  (CallFunction|ApplyDamagetoActor ActorToExplode)
  (CallFunction|LaunchActor ActorToExplode)
  (Development|PrintString (Utilities|GetDisplayName ActorToExplode) true true "(R=1.000000,G=0.044237,B=0.000000,A=1.000000)" 5.0))

(fn HandleCustomImpulseFromDamage
  (params (ImpulseValue :type Vector :default "0, 0, 0") (Damage :type Float%20%28single-precision%29) (DamageTags :type Gameplay%20Tag%20Container%20Structure%20%28by%20ref%29) (HitResult :type Hit%20Result%20Structure%20%28by%20ref%29) (InstigatingController :type Controller%20Object%20Reference) (DamageCauser :type Actor%20Object%20Reference) (EffectContext :type Gameplay%20Effect%20Context%20Handle%20Structure%20%28by%20ref%29))
  (return false))

(fn IsTrainVault
  (return false))

(fn IsVaultStarted
  (return false))

(fn LaunchActor
  (params (Actor%20To%20Launch :type Actor%20Object%20Reference))
  (bind _actor_to_launch (Variables|GetActor%20To%20Launch))
  (bind _asfort_pawn (Utilities|Casting|CastToFortPawn Actor%20To%20Launch)
    (:then
      ;[1 1 1] Launch Player
      (if (not (or (GameplayTags|HasMatchingGameplayTag _asfort_pawn "(TagName=\"Gameplay.Status.KnockbackImmunity\")") (GameplayTags|HasMatchingGameplayTag _asfort_pawn "(TagName=\"Gameplay.Action.Player.AthenaInVehicle\")")))
        (Character|LaunchCharacter _asfort_pawn (Utilities|CalculateLaunchVelocity (Transformation|GetActorLocation (Variables|GetActor%20To%20Launch)) (Transformation|GetActorLocation)) true true)))
    (:CastFailed
      ;[1 1 1] LaunchVehicle
      (bind _asfort_vehicle_interface (Utilities|Casting|CastToFortVehicleInterface _actor_to_launch))
      (bind _returnvalue (Vehicle|GetVehicleActor _asfort_vehicle_interface :!class /Script/FortniteGame.FortVehicleInterface))
      (Audio|BroadcastAbilityImpactatLocation 0 _returnvalue (Math|Vector|MakeVector (.x (* (Math|Vector|Normalize (select (>= (Math|Vector|VectorLength (Transformation|GetVelocity _returnvalue)) 5e+02) (Transformation|GetVelocity _returnvalue) (- (Transformation|GetActorLocation _actor_to_launch) (Transformation|GetActorLocation)))) (ScalableFloat|GetValueatLevel (Variables|Default|GetVehicleImpulseStrength)))) (.y (* (Math|Vector|Normalize (select (>= (Math|Vector|VectorLength (Transformation|GetVelocity _returnvalue)) 5e+02) (Transformation|GetVelocity _returnvalue) (- (Transformation|GetActorLocation _actor_to_launch) (Transformation|GetActorLocation)))) (ScalableFloat|GetValueatLevel (Variables|Default|GetVehicleImpulseStrength)))) (ScalableFloat|GetValueatLevel (Variables|Default|GetVehicleImpulseStrengthZ))) 777.0))))

(fn LocalOnFailedInteract
  (params (InteractingPawn :type Fort%20Player%20Pawn%20Object%20Reference)))

(fn MinigameIsInProgress
  (params (IgnoreWarmUp :type Boolean))
  (bind _self self)
  (bind _returnvalue (Minigame|GetMinigame _self :!class /Script/FortniteGame.MinigameBlueprintLibrary))
  (Utilities|IsValid _returnvalue
    (:Is%20Valid
      (Utilities|IsValid _returnvalue
        (:Is%20Valid
          ;[1 1 1] Is in progress pure function includes Warmup time, optionally ignore that and only use "In Progress" time
          (if IgnoreWarmUp
            (return true (Utilities|Enum|Equal(Enum) (Variables|/Script/FortniteGame.FortMinigame|GetCurrentState _returnvalue) "InProgress"))
            (else
              (return true (Minigame|IsinProgress _returnvalue)))))
        (:Is%20Not%20Valid
          _)))
    (:Is%20Not%20Valid
      _))
  (return false false))

(fn OnRep_DamageStateNotify
  ;[1 1 1] Use Current Health to figure out which emitters to spawn
  (bind (_array_element _array_index) (Utilities|Array|ForEachLoopwithBreak (Variables|DamageBasedEmitters|GetDamageStateParticleSystems))
    (:LoopBody
      (if (.Spawned_19_ADE1ECD74E521455DA9695B1FDB47581 (Utilities|Struct|BreakAthenaDamageStateParticles _array_element))
        _
        (elif (<= (Health|GetHealthPercent) (.HealthPercent_24_AC35B253419518DA53906DB5CCAD79F8 (Utilities|Struct|BreakAthenaDamageStateParticles _array_element)))
          ;[1 1 1] Set that this element has spawned (to avoid doing it twice)
          (Utilities|Array|SetArrayElem (Variables|DamageBasedEmitters|GetDamageStateParticleSystems) _array_index (Utilities|Struct|MakeAthenaDamageStateParticles (Utilities|Struct|BreakAthenaDamageStateParticles _array_element) true (.LocalSpawnPosition_9_507DFDCB4A99E944783EEAB44C495149 (Utilities|Struct|BreakAthenaDamageStateParticles _array_element)) (.LocalSpawnRotation_12_86D59DAF4DC65EF2E0CEB2BD7740F557 (Utilities|Struct|BreakAthenaDamageStateParticles _array_element)) (.LocalSpawnScale_20_D001987D4D7EB90DF56AAC9B5F8133A8 (Utilities|Struct|BreakAthenaDamageStateParticles _array_element))))
          ;[1 1 1] Spawn the effect
          (Development|PrintString (Utilities|GetDisplayName (Utilities|Struct|BreakAthenaDamageStateParticles _array_element)) true true "(R=1.000000,G=0.000000,B=0.089557,A=1.000000)" 1e+01)
          (bind _returnvalue (AddComponent|NODE_AddParticleSystemComponent-0 :RelativeTransform_Location (.LocalSpawnPosition_9_507DFDCB4A99E944783EEAB44C495149 (Utilities|Struct|BreakAthenaDamageStateParticles _array_element)) :RelativeTransform_Rotation (.LocalSpawnRotation_12_86D59DAF4DC65EF2E0CEB2BD7740F557 (Utilities|Struct|BreakAthenaDamageStateParticles _array_element)) :RelativeTransform_Scale (.LocalSpawnScale_20_D001987D4D7EB90DF56AAC9B5F8133A8 (Utilities|Struct|BreakAthenaDamageStateParticles _array_element))))
          (Effects|Components|ParticleSystem|SetTemplate _returnvalue (Utilities|Struct|BreakAthenaDamageStateParticles _array_element))
          (Utilities|Array|Add (Variables|DamageBasedEmitters|GetSpawnedDamageStateEmitters) _returnvalue))))))

(fn OnRep_ExplosiveFxRepNotify
  (Projectile|RemoveGameplayCue%28Looping%29Local self (Variables|Default|GetFireCue) (Ability|GameplayCue|MakeGameplayCueParameters 0.0))
  (GameplayCue|ExecuteGameplayCue%28Burst%29Local self (Variables|Default|GetGC_FinialThermiteExplosion) (Ability|GameplayCue|MakeGameplayCueParameters 0.0) false))

(fn OnRep_FirstExplosiveRepNotify
  (GameplayCue|ExecuteGameplayCue%28Burst%29Local self (Variables|Default|GetGC_FirstThermiteExplosion) (Ability|GameplayCue|MakeGameplayCueParameters 0.0) false))

(fn RequestTrainVaultStart
  (params (Payload :type Struct%20Vivid%20Razor%20Core%20Vault%20Start%20Payload%20Structure))
  (return false))

(fn RequestVaultStart
  (params (Payload :type Struct%20Vivid%20Razor%20Core%20Vault%20Start%20Payload%20Structure))
  (bind (_vaultlocationtag_24_4638c5e048c21e5b7f1d909063318292 _triggeringpawn_33_567a525a4447d3f140bec19d19506604 _triggeringactor_32_c6d5ffac409c2b65b7d64d89bf7cf0c5) (Utilities|Struct|BreakStructVividRazorCoreVaultStartPayload Payload))
  (if (Variables|Default|GetIsTopOpen)
    (CallFunction|StartFuse (Pawn|GetController (Utilities|ResolveSoftReference _triggeringpawn_33_567a525a4447d3f140bec19d19506604)) (Utilities|ResolveSoftReference _triggeringactor_32_c6d5ffac409c2b65b7d64d89bf7cf0c5))
    (return false)))

(fn ShouldCancelExplosion
  ;[1 0.08217 0.06227] Don't allow the prop to explode if it is in a valid minigame that is not running.
  (bind (_hasvalidminigame _isrunning) (Default|MinigameIsinProgress false))
  (return (and _hasvalidminigame (not _isrunning))))

(fn ShouldDie
  (params (Damage :type Float%20%28single-precision%29) (EventInstigator :type Controller%20Object%20Reference) (DamageCauser :type Actor%20Object%20Reference) (bIsTest :type Boolean) (EffectContext :type Gameplay%20Effect%20Context%20Handle%20Structure%20%28by%20ref%29))
  ;[1 1 1] This Function get's called on Death. Health Is Zero get's set to true so that the damage event triggers the Death Fx. Should die is set to false so that the actors collision stays intact
  (Variables|Bools|SetHealthIsZero (Math|Float|NearlyEqual(Float) (Health|GetHealthPercent) 0.0 0.001))
  (return (Variables|Default|GetbShouldDie)))

(fn SubmitAnalytics
  (params (TrainVaultOpened :type Boolean))
  (Networking|SwitchHasAuthority
    (:Authority
      (if (and (not (Replay|GetIsPlayingReplay)) (not (Variables|Analytics|GetAnalyticsSubmitted)))
        (Variables|Analytics|SetAnalyticsSubmitted true)
        (Utilities|FlowControl|Sequence
          (:then_0
            (Utilities|FlowControl|Sequence
              (:then_0
                (bind _returnvalue (Analytics|ConstructAnalyticsAttribute :AttributeName "VaultOpened" :IntValue -1 :floatValue -1.0 :BoolValue (Variables|GetTrainVaultOpened)))
                (Utilities|Array|Add (LocalVariables|GetAnalyticsEventAttributes) _returnvalue)
                (bind _returnvalue_2 (Analytics|ConstructAnalyticsAttribute "VaultLocation" (Utilities|String|ToString(Vector) (Transformation|GetActorLocation))))
                (Utilities|Array|Add (LocalVariables|GetAnalyticsEventAttributes) _returnvalue_2)
                (bind _returnvalue_3 (Analytics|ConstructAnalyticsAttribute :AttributeName "VaultStartTime" :IntValue -1 :floatValue (Variables|Analytics|GetStartTime) :BoolValue false))
                (Utilities|Array|Add (LocalVariables|GetAnalyticsEventAttributes) _returnvalue_3)
                (bind _returnvalue_4 (Analytics|ConstructAnalyticsAttribute :AttributeName "VaultResultTime" :IntValue -1 :floatValue (Fortnite|GetRealMatchTime) :BoolValue false))
                (Utilities|Array|Add (LocalVariables|GetAnalyticsEventAttributes) _returnvalue_4))
              (:then_2
                (LocalVariables|SetVaultStarterID_String "INVALID")
                (bind _fusekiller (Variables|Default|GetFuseKiller)
                  (:then
                    (bind _playerstate (Variables|/Script/Engine.Controller|GetPlayerState _fusekiller)
                      (:then
                        (bind _asfort_player_state (Utilities|Casting|CastToFortPlayerState _playerstate)
                          (:then
                            (bind _returnvalue_5 (PlayerInfo|GetDebugStringforUniqueId (PlayerState|GetPlayerUniqueNetID _asfort_player_state)))
                            (LocalVariables|SetVaultStarterID_String _returnvalue_5))
                          (:CastFailed
                            _))
                        (bind _returnvalue_6 (Analytics|ConstructAnalyticsAttribute "VaultStarterID" (LocalVariables|GetVaultStarterID_String)))
                        (Utilities|Array|Add (LocalVariables|GetAnalyticsEventAttributes) _returnvalue_6))
                      (:else
                        _)))
                  (:else
                    _)))))
          (:then_1
            (Analytics|FireAnalyticsEventBlueprintArray "Core.VividRazor.TrainVault" (LocalVariables|GetAnalyticsEventAttributes))))))))

(fn UpdateThermiteMaterials
  (params (ShowRealMesh :type Boolean))
  (if ShowRealMesh
    (Rendering|SetVisibility (Variables|Debug|GetStaticMeshComponent) true true)
    (Rendering|Material|SetMaterial (Variables|Debug|GetStaticMeshComponent) 0 (Variables|Default|GetMI_Thermite_0))
    (elif (ScalableFloat|GetValueasBool (Variables|Default|GetThermiteHologramEnabled))
      (Rendering|Material|SetMaterial (Variables|Debug|GetStaticMeshComponent) 0 (Variables|Default|GetMI_Thermite_Hologram))
      (else
        ; FORT-844553: Hologram isn't enabled (primarily for capture team, hide the thermite on the vault
        (Rendering|SetVisibility (Variables|Debug|GetStaticMeshComponent) false true)))))

(fn ConstructionScript
  (|CallParentFunction))
