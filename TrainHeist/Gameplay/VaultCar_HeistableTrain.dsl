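; Vault car of the heistable train, where a thermite hit near the explosives prop starts a fuse to blow the vault.
; Locked-in detonation stops the train, with replicated toggles driving braking cues on every car.
; Opening swaps the door mesh, injects a chest level, spawns a navmesh and reinforcements actor, and optionally rifts out players inside.
; Boarding, detonation and encounter events advance replicated music stages, while boss events notify the music component.
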(event EventEndPlay
  (params (EndPlayReason :type EEndPlayReason%20Enum :default "Destroyed"))
  (Utilities|FlowControl|Sequence
    (:then_0
      (Networking|SwitchHasAuthority
        (:Remote
          (for _array_element (Variables|/Script/TrainCoreRuntime.FortTrainManager|GetCars (CallFunction|GetTrainManager))
            (Utilities|IsValid (Utilities|Struct|BreakTrainCarRuntimeData _array_element)
              (:Is%20Valid
                (Projectile|RemoveGameplayCue%28Looping%29Local (Utilities|Struct|BreakTrainCarRuntimeData _array_element) (Variables|Default|GetGC_TrainBreaking_IndividualCar) (Ability|GameplayCue|MakeGameplayCueParameters 0.0))))))))
    (:then_1
      (EventDispatchers|UnbindEventfromOnPawnBasedonMovingPlatform (Variables|Default|GetFortTrainMovingPlatform) (EventDispatchers|CreateEvent :!event Custom|PlayerBoardedVault))
      (EventDispatchers|UnbindEventfromOnPawnDebasedMovingPlatform (Variables|Default|GetFortTrainMovingPlatform) (EventDispatchers|CreateEvent :!event Custom|PlayerUnboardedVault))
      (Utilities|Array|Clear (Variables|Default|GetPawnsOnVaultCar)))))

(event EventOnActorTargetedForImpulse
  (params (Actor :type Actor%20Object%20Reference))
  (bind _asfort_pawn (Utilities|Casting|CastToFortPawn Actor))
  (Character|LaunchCharacter _asfort_pawn (* (Math|Vector|Normalize (- (Transformation|GetActorLocation) (Transformation|GetActorLocation Actor))) (ScalableFloat|GetValueatLevel (Variables|Default|GetSF_ImpulseValue))) true true))

(event Custom|CheckForPlayersInGap
  (Networking|SwitchHasAuthority
    (:Authority
      (bind _outactors (Collision|SphereOverlapActors (Utilities|Operators|Add (Transformation|GetActorLocation) (* (Transformation|GetActorForwardVector) -668.0) "0.000000,0.000000,150.000000") 2e+02 (Utilities|Array|MakeArray "ObjectTypeQuery3") "/Script/FortniteGame.FortPawn"))
      (for _array_element _outactors
        (bind _asfort_pawn (Utilities|Casting|CastToFortPawn _array_element))
        (Pawn|Character|LaunchCharacterJump _asfort_pawn (+ (* (Math|Vector|CrossProduct (Math|Vector|Normalize (Transformation|GetVelocity)) "0.000000,0.000000,1.000000") 1e+03) "0.000000,0.000000,1000.000000") true true false true)))))

(event Custom|OnFuseStarted
  (CallFunction|SetOverheatVisuals false)
  (Class|PropVaultExplosivesHeistableTrain|PlayAlarmAudio (stale EventBeginPlay _asprop_vault_explosives_heistable_train) false))

(event Custom|OnStopDetonation
  (CallFunction|SetOverheatVisuals true)
  (Class|PropVaultExplosivesHeistableTrain|PlayAlarmAudio (stale EventBeginPlay _asprop_vault_explosives_heistable_train) false))

(event Custom|OnDetonationLockedIn
  (Class|PropVaultExplosivesHeistableTrain|PlayAlarmAudio (stale EventBeginPlay _asprop_vault_explosives_heistable_train) true)
  (Class|FortTrainManager|BeginStoppingTrain (CallFunction|GetTrainManager))
  (Networking|SwitchHasAuthority
    (:Authority
      (Variables|Default|SetMusic%20Stage "NewEnumerator2"))))

(event EventBeginPlay
  (bind _returnvalue (CallFunction|GetTrainManager))
  (bind _forttrainmovingplatform (Variables|Default|GetFortTrainMovingPlatform))
  (|CallParentFunction)
  (Networking|SwitchHasAuthority
    (:Authority
      (Utilities|FlowControl|Sequence
        (:then_0
          (Class|FortAsyncActionCurrentPlaylistReady|PlaylistDataReadyAsync
            (:OnReady
              (Utilities|FlowControl|Delay 0.2)
              (CallFunction|ReserveAIDensity))))
        (:then_1
          _)))
    (:Remote
      _))
  (Utilities|FlowControl|Delay 5.0)
  (Utilities|FlowControl|Sequence
    (:then_0
      (bind _asprop_vault_explosives_heistable_train (Utilities|Casting|CastToProp_VaultExplosives_HeistableTrain (Variables|/Script/Engine.ChildActorComponent|GetChildActor (Variables|Default|GetProp_VaultExplosives_HeistableTrain))))
      (Class|PropVaultExplosivesHeistableTrain|Init _asprop_vault_explosives_heistable_train self)
      (Default|BindEventtoOnDetonationLockedIn _asprop_vault_explosives_heistable_train (delegate Custom|OnDetonationLockedIn))
      (Default|BindEventtoOnFuseStarted _asprop_vault_explosives_heistable_train (delegate Custom|OnFuseStarted))
      (Default|BindEventtoOnStopDetonation _asprop_vault_explosives_heistable_train (delegate Custom|OnStopDetonation))
      (Default|BindEventtoOnDetonation _asprop_vault_explosives_heistable_train (delegate Custom|OnStopDetonation)))
    (:then_1
      (EventDispatchers|BindEventtoOnPawnBasedonMovingPlatform _forttrainmovingplatform (delegate Custom|PlayerBoardedVault))
      (EventDispatchers|BindEventtoOnPawnDebasedMovingPlatform _forttrainmovingplatform (delegate Custom|PlayerUnboardedVault)))
    (:then_2
      (EventDispatchers|BindEventtoOnTrainAbouttoStopEvent _returnvalue (delegate Custom|OnTrainAboutToStop))
      (EventDispatchers|BindEventtoOnTrainStoppedEvent _returnvalue (delegate Custom|OnTrainStopped))
      (Default|SetMusicStage (Variables|Default|GetNPCMusicComponent_VividRazorVaults) (Variables|Default|GetMusic%20Stage))
      (for _array_element (Variables|/Script/TrainCoreRuntime.FortTrainManager|GetCars (CallFunction|GetTrainManager))
        (Utilities|IsValid (Class|FortBaseMovingPlatformComponent|FindMovingPlatformComponentfromActor (Utilities|Struct|BreakTrainCarRuntimeData _array_element))
          (:Is%20Valid
            (EventDispatchers|BindEventtoOnPawnBasedonMovingPlatform (Class|FortBaseMovingPlatformComponent|FindMovingPlatformComponentfromActor (Utilities|Struct|BreakTrainCarRuntimeData _array_element)) (EventDispatchers|CreateEvent :!event Custom|PlayerOnTrain))))))))

(event Custom|PlayerBoardedVault
  (params (MovingPlatform :type Fort%20Base%20Moving%20Platform%20Component%20Object%20Reference) (Pawn :type Fort%20Pawn%20Object%20Reference))
  (Utilities|Array|Add (Variables|Default|GetPawnsOnVaultCar) Pawn))

(event Custom|PlayerUnboardedVault
  (params (MovingPlatform :type Fort%20Base%20Moving%20Platform%20Component%20Object%20Reference) (Pawn :type Fort%20Pawn%20Object%20Reference) (NewBase :type Actor%20Object%20Reference))
  (Utilities|Array|RemoveItem (Variables|Default|GetPawnsOnVaultCar) Pawn))

(event Custom|OnTrainStopped
  (Networking|SwitchHasAuthority
    (:Authority
      (Variables|Default|SetToggleBreakingFX false))))

(event Custom|OnTrainAboutToStop
  (Networking|SwitchHasAuthority
    (:Authority
      (Variables|Default|SetToggleBreakingFX true)
      (CallFunction|PlayStoppingAudio))))

(event Custom|OpenValut_MultiCast
  (Utilities|FlowControl|Delay 0.2)
  (Components|StaticMesh|SetStaticMesh (Variables|Default|GetValutDoor) (Variables|Default|GetBlownUpVaultDoor))
  (Utilities|Time|SetTimerbyEvent (delegate Custom|SpawnChests) 1.5)
  (CallFunction|SpawnNavigationData)
  (Networking|SwitchHasAuthority
    (:Authority
      (if (ScalableFloat|GetValueasBool (Variables|Default|GetSF_TeleportPlayersInVaultWhenTrainStopped))
        (bind _outhits (Collision|MultiBoxTraceByChannel :Start (+ (Transformation|GetActorLocation) "0.000000,0.000000,215.000000") :End (+ (Transformation|GetActorLocation) "0.000000,0.000000,215.000000") :HalfSize "850.000000,170.000000,135.000000" :Orientation (Transformation|GetActorRotation) :TraceChannel "TraceTypeQuery9" :bTraceComplex false :DrawDebugType "Persistent" :bIgnoreSelf true :TraceColor "(R=1.000000,G=0.000000,B=0.000000,A=1.000000)" :TraceHitColor "(R=0.000000,G=1.000000,B=0.000000,A=1.000000)" :DrawTime 5.0))
        (for _array_element _outhits
          (bind _asfort_player_pawn (Utilities|Casting|CastToFortPlayerPawn (.HitActor (Collision|BreakHitResult _array_element))))
          (bind _output_get (Variables|Default|SetActorsToIgnoreRift (Variables|Default|GetPawnsOnVaultCar)))
          (Utilities|Array|RemoveItem _output_get _asfort_player_pawn)
          (bind (_returnvalue_location _returnvalue_rotation) (Math|Transform|BreakTransform (Transformation|GetWorldTransform (Variables|/Script/Engine.Character|GetMesh _asfort_player_pawn))))
          (bind _returnvalue (Game|SpawnActorfromClass "/Game/Athena/Items/Consumables/RiftItem/BGA_RiftPortal_Item_Athena.BGA_RiftPortal_Item_Athena_C" (Math|Transform|MakeTransform _returnvalue_location _returnvalue_rotation) "AdjustIfPossibleButAlwaysSpawn" "OverrideRootScale" 0 (Variables|Default|GetActorsToIgnoreRift) 0 _asfort_player_pawn))
          (Utilities|Array|Add (Variables|Default|GetSpawnedRiftPortals) _returnvalue))
        (Utilities|Time|SetTimerbyEvent (delegate Custom|DestroyRifts) (ScalableFloat|GetValueatLevel (Variables|Default|GetSF_DestroyRiftsTime)))
        (Utilities|Array|Clear (Variables|Default|GetActorsToIgnoreRift))))))

(event Custom|SpawnChests
  (Networking|SwitchHasAuthority
    (:Authority
      (Class|FortLevelInjectionManagerBase|RequesttoInjectLevel (Actor|GetComponentbyClass (Fortnite|Athena|GameState|GetGameStateAthena) "/Script/FortLevelInjectingRuntime.FortLevelInjectionManagerBase") (Utilities|ResolveSoftReference (Variables|Default|GetVaultLevelInjectionData)) (Math|Transform|MakeTransform (+ (.location (Transformation|GetWorldTransform (Variables|Default|GetStaticMesh))) "0.000000,0.000000,80.000000") (.rotation (Transformation|GetWorldTransform (Variables|Default|GetStaticMesh))))))))

(event Custom|SpawnNavigationData
  (bind (_returnvalue_location _returnvalue_rotation _returnvalue_scale) (Math|Transform|BreakTransform (Transformation|GetActorTransform self)))
  ; Handle Spawning NavMesh on top of Cars and Spawning the Reinforcements
  (Networking|SwitchHasAuthority
    (:Authority
      (bind _returnvalue (Game|SpawnActorfromClass "/TrainHeist/Navigation/BGA_VaultCar_NavGeneration.BGA_VaultCar_NavGeneration_C" (Math|Transform|MakeTransform (+ (+ (Transformation|GetActorLocation self) (* (Transformation|GetActorForwardVector self) -311.0)) (* (Transformation|GetActorUpVector self) 37.0)) _returnvalue_rotation _returnvalue_scale) "AlwaysSpawn" "MultiplyWithRoot" self 0 0)) ; Spawn BGA that will generate NavMesh on top of Train only when the Train is stopped
      (CallFunction|BindMusicEvents _returnvalue))))

(event Custom|DestroyRifts
  (for _array_element (Variables|Default|GetSpawnedRiftPortals)
    (Class|BGARiftPortalItemAthena|DestroyRift _array_element))
  (Utilities|Array|Clear (Variables|Default|GetSpawnedRiftPortals)))

(event Custom|Bind%20Music%20Events
  (params (Vault%20Car%20Nav%20Generation :type BGA%20Vault%20Car%20Nav%20Generation%20Object%20Reference))
  ; Encounter-based music events
  (EventDispatchers|BindEventtoOnEncounterCompleted Vault%20Car%20Nav%20Generation (delegate Custom|Encounter%20Completed))
  (bind _bp_dynamicreinforcementsmanager (Variables|/TrainHeist/Navigation/BGA_VaultCar_NavGeneration.BGA_VaultCar_NavGeneration_C|GetBP_DynamicReinforcementsManager Vault%20Car%20Nav%20Generation)
    (:then
      (Default|BindEventtoOnBossSpawned _bp_dynamicreinforcementsmanager (delegate Custom|Boss%20Spawned))
      (Default|BindEventtoOnBossDied _bp_dynamicreinforcementsmanager (delegate Custom|Boss%20Died)))
    (:else)))

(event Custom|Encounter%20Completed
  ; Encounter-based music events
  (Variables|Default|SetMusic%20Stage "NewEnumerator3"))

(event Custom|Boss%20Spawned
  ; Encounter-based music events
  (bind _npcmusiccomponent_vividrazorvaults (Variables|Default|GetNPCMusicComponent_VividRazorVaults)
    (:then
      (Class|NPCMusicComponentVividRazorVaults|BossSpawned _npcmusiccomponent_vividrazorvaults))
    (:else)))

(event Custom|Boss%20Died
  ; Encounter-based music events
  (bind _npcmusiccomponent_vividrazorvaults (Variables|Default|GetNPCMusicComponent_VividRazorVaults)
    (:then
      (Class|NPCMusicComponentVividRazorVaults|BossDefeated _npcmusiccomponent_vividrazorvaults))
    (:else)))

(event Custom|PlayStoppingAudio
  ;[0 0 0 0.5] Stopping Audio
  (Networking|CanExecuteCosmeticEvents
    (:True
      (Audio|SpawnSoundAttached (Variables|Default|GetStoppingAudio) (Variables|Default|GetStaticMesh) "None" "-208.000000,32.000000,232.000000" "-90.000000,0.000000,0.000000"))))

(event Custom|PlayerOnTrain
  (params (MovingPlatform :type Fort%20Base%20Moving%20Platform%20Component%20Object%20Reference) (Pawn :type Fort%20Pawn%20Object%20Reference))
  (bind (_asfort_player_pawn_athena _bsuccess) (Utilities|Casting|CastToFortPlayerPawnAthena Pawn))
  (Networking|SwitchHasAuthority
    (:Authority
      (Utilities|FlowControl|DoOnce false
        (:Completed
          (if (and _bsuccess (Utilities|Enum|Equal(Enum) (Variables|Default|GetMusic%20Stage) "NewEnumerator0"))
            (Variables|Default|SetMusic%20Stage "NewEnumerator1")))))))

(fn GetMusicStage
  (return (Variables|Default|GetMusic%20Stage)))

(fn IsTrainVault
  (return true))

(fn IsVaultStarted
  (return false))

(fn OnRep_MusicStage
  (bind _npcmusiccomponent_vividrazorvaults (Variables|Default|GetNPCMusicComponent_VividRazorVaults)
    (:then
      (Class|NPCMusicComponentVividRazorVaults|TransitionToNextMusicStage _npcmusiccomponent_vividrazorvaults (Variables|Default|GetMusic%20Stage)))
    (:else)))

(fn OnRep_ToggleBreakingFX
  (Networking|SwitchHasAuthority
    (:Remote
      (if (Variables|Default|GetToggleBreakingFX)
        (for _array_element (Variables|/Script/TrainCoreRuntime.FortTrainManager|GetCars (CallFunction|GetTrainManager))
          (Utilities|IsValid (Utilities|Struct|BreakTrainCarRuntimeData _array_element)
            (:Is%20Valid
              (Projectile|AddGameplayCue%28Looping%29Local (Utilities|Struct|BreakTrainCarRuntimeData _array_element) (Variables|Default|GetGC_TrainBreaking_IndividualCar) (Ability|GameplayCue|MakeGameplayCueParameters 0.0)))))
        (else
          (for _array_element_2 (Variables|/Script/TrainCoreRuntime.FortTrainManager|GetCars (CallFunction|GetTrainManager))
            (Utilities|IsValid (Utilities|Struct|BreakTrainCarRuntimeData _array_element_2)
              (:Is%20Valid
                (Projectile|RemoveGameplayCue%28Looping%29Local (Utilities|Struct|BreakTrainCarRuntimeData _array_element_2) (Variables|Default|GetGC_TrainBreaking_IndividualCar) (Ability|GameplayCue|MakeGameplayCueParameters 0.0))))))))))

(fn RequestTrainVaultStart
  (params (Payload :type Struct%20Vivid%20Razor%20Core%20Vault%20Start%20Payload%20Structure))
  (bind (_vaultlocationtag_24_4638c5e048c21e5b7f1d909063318292 _triggeringpawn_33_567a525a4447d3f140bec19d19506604 _triggeringactor_32_c6d5ffac409c2b65b7d64d89bf7cf0c5 _triggeringactorimpactnormal_28_7604343b448314764cd2bb8091ff0a9c) (Utilities|Struct|BreakStructVividRazorCoreVaultStartPayload Payload))
  (bind _output (Utilities|ResolveSoftReference _triggeringactor_32_c6d5ffac409c2b65b7d64d89bf7cf0c5))
  (bind (_payload_vaultlocationtag_24_4638c5e048c21e5b7f1d909063318292 _payload_triggeringpawn_33_567a525a4447d3f140bec19d19506604 _payload_triggeringactor_32_c6d5ffac409c2b65b7d64d89bf7cf0c5) (Utilities|Struct|BreakStructVividRazorCoreVaultStartPayload (Variables|GetPayload_VaultLocationTag_24_4638C5E048C21E5B7F1D909063318292)))
  (if (> (Math|Vector|DotProduct _triggeringactorimpactnormal_28_7604343b448314764cd2bb8091ff0a9c (Math|Vector|VectorUp)) 0)
    (Utilities|IsValid _output
      (:Is%20Valid
        (Utilities|IsValid (Variables|/Script/Engine.ChildActorComponent|GetChildActor (Variables|Default|GetProp_VaultExplosives_HeistableTrain))
          (:Is%20Valid
            (if (<= (Math|Vector|VectorLength (- (Transformation|GetActorLocation _output) (Transformation|GetActorLocation (Variables|/Script/Engine.ChildActorComponent|GetChildActor (Variables|Default|GetProp_VaultExplosives_HeistableTrain))))) (ScalableFloat|GetValueatLevel (Variables|Default|GetSF_TriggerThermiteExplosiveDistanceCheck)))
              (bind _asprop_vault_explosives_heistable_train (Utilities|Casting|CastToProp_VaultExplosives_HeistableTrain (Variables|/Script/Engine.ChildActorComponent|GetChildActor (Variables|Default|GetProp_VaultExplosives_HeistableTrain)))
                (:then
                  (if (not (Variables|/TrainHeist/Gameplay/Prop_VaultExplosives_HeistableTrain.Prop_VaultExplosives_HeistableTrain_C|GetFuseStarted _asprop_vault_explosives_heistable_train))
                    (Class|PropVaultExplosivesHeistableTrain|StartFuse _asprop_vault_explosives_heistable_train (Pawn|GetController (Utilities|ResolveSoftReference _payload_triggeringpawn_33_567a525a4447d3f140bec19d19506604)) (Utilities|ResolveSoftReference _payload_triggeringactor_32_c6d5ffac409c2b65b7d64d89bf7cf0c5))
                    (return false)))
                (:CastFailed
                  _))
              (return true)
              (else
                _)))
          (:Is%20Not%20Valid
            _)))
      (:Is%20Not%20Valid
        _))
    (else
      _)))

(fn RequestVaultStart
  (params (Payload :type Struct%20Vivid%20Razor%20Core%20Vault%20Start%20Payload%20Structure))
  (return false))

(fn ReserveAIDensity
  (bind _returnvalue (Class|FortAthenaLivingWorldManager|GetLivingWorldManager))
  (bind _self self)
  (Utilities|IsValid _returnvalue
    (:Is%20Valid
      (bind _returnvalue_2 (Class|FortAthenaLivingWorldManager|RegisterDensityReservoir _returnvalue _self 4.0 12001.0))
      (if _returnvalue_2
        _))
    (:Is%20Not%20Valid
      (Fortnite|FortShippingLog _self ":: Trying to get Living World Manager, but returned Invalid / Null" true))))

(fn SetOverheatVisuals
  (params (Reverse :type Boolean))
  (bind _valutdoor (Variables|Default|GetValutDoor))
  (Networking|CanExecuteCosmeticEvents
    (:True
      (for (_ _array_index) (Rendering|Material|GetMaterials _valutdoor)
        (Rendering|Material|SetScalarParameterValue (Class|BTechArtGlobalLibrary|GetDynamicMaterialInstance _valutdoor _array_index) "OverheatTimestampInvert" (Math|Conversions|ToFloat(Boolean) Reverse) :!class /Script/Engine.MaterialInstanceDynamic)
        (Rendering|Material|SetScalarParameterValue (Class|BTechArtGlobalLibrary|GetDynamicMaterialInstance _valutdoor _array_index) "OverheatTimestamp" (Utilities|Time|GetGameTimeinSeconds) :!class /Script/Engine.MaterialInstanceDynamic)))))

(fn ConstructionScript
  (|CallParentFunction))
